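# Physics asset for the left skirt of the F_MED_GrowMallet female character body cosmetic.
# Front and side chains of four capsule bodies each hang from tiny base-bone capsules, with radii generally widening down the chain.
# Constraints keep swing to a few degrees or zero and lock twist, except at the front tip.
# An unconstrained left calf capsule is included, likely for leg collision, while the thigh body has no shape.
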PhysicsAsset /BRCosmetics/Characters/Player/Female/Medium/Bodies/F_MED_GrowMallet/Meshes/F_MED_GrowMallet_LeftSkirt_Physics
bodies {
 base_skirt_fr_l [{shapeType=Capsule, length=0.505, radius=0.505, shapeName=base_skirt_fr_l_capsule, rotation=<rot 0 -90 0>}]
 base_skirt_sd_l [{shapeType=Capsule, length=0.505, radius=0.505, shapeName=base_skirt_sd_l_capsule, rotation=<rot 0 -90 0>}]
 calf_l [{
   shapeType=Capsule
   length=33.3659
   radius=7.0424
   shapeName=calf_l_capsule
   center=<-22.2961 -2.1523 -0.571>
   rotation=<rot -0.1636 92.8952 93.2305>
  }]
 dyn_skirt_fr_01_l [{shapeType=Capsule, length=6, radius=5, shapeName=dyn_skirt_fr_01_l_capsule, center=<7 0 4.5>, rotation=<rot 82.564 90.0575 0.0006>}]
 dyn_skirt_fr_02_l [{shapeType=Capsule, length=10.7702, radius=6, shapeName=dyn_skirt_fr_02_l_capsule, center=<6 0 4>, rotation=<rot 84.8227 89.9615 0.0005>}]
 dyn_skirt_fr_03_l [{shapeType=Capsule, length=12.3431, radius=7, shapeName=dyn_skirt_fr_03_l_capsule, center=<6.5 0 3>, rotation=<rot 90 90 0>}]
 dyn_skirt_fr_04_l [{shapeType=Capsule, length=11.221, radius=7, shapeName=dyn_skirt_fr_04_l_capsule, center=<9 0 2>, rotation=<rot 84.3248 90.0326 0.0007>}]
 dyn_skirt_sd_01_l [{shapeType=Capsule, length=6, radius=4.5, shapeName=dyn_skirt_sd_01_l_capsule, center=<6 0 3>, rotation=<rot 90 90 0>}]
 dyn_skirt_sd_02_l [{shapeType=Capsule, length=6, radius=5, shapeName=dyn_skirt_sd_02_l_capsule, center=<6 0 3>, rotation=<rot 90 90 0>}]
 dyn_skirt_sd_03_l [{shapeType=Capsule, length=6, radius=5.5, shapeName=dyn_skirt_sd_03_l_capsule, center=<6 0 3>, rotation=<rot 90 90 0>}]
 dyn_skirt_sd_04_l [{shapeType=Capsule, length=6, radius=6, shapeName=dyn_skirt_sd_04_l_capsule, center=<9 0 2.3797>, rotation=<rot 90 90 0>}]
 thigh_l []
}
constraints [
 {bone1Name=dyn_skirt_fr_01_l, bone2Name=base_skirt_fr_l, swing1LimitDegrees=2, swing2LimitDegrees=5, twistLimitDegrees=0}
 {bone1Name=dyn_skirt_fr_02_l, bone2Name=dyn_skirt_fr_01_l, swing1LimitDegrees=2, swing2LimitDegrees=4, twistLimitDegrees=0}
 {bone1Name=dyn_skirt_fr_03_l, bone2Name=dyn_skirt_fr_02_l, swing1LimitDegrees=2, swing2LimitDegrees=5, twistLimitDegrees=0}
 {bone1Name=dyn_skirt_fr_04_l, bone2Name=dyn_skirt_fr_03_l, swing1LimitDegrees=4, swing2LimitDegrees=5, twistLimitDegrees=6}
 {bone1Name=dyn_skirt_sd_01_l, bone2Name=base_skirt_sd_l, swing1LimitDegrees=0, swing2LimitDegrees=2, twistLimitDegrees=0}
 {bone1Name=dyn_skirt_sd_02_l, bone2Name=dyn_skirt_sd_01_l, swing1LimitDegrees=0, swing2LimitDegrees=2, twistLimitDegrees=0}
 {bone1Name=dyn_skirt_sd_03_l, bone2Name=dyn_skirt_sd_02_l, swing1LimitDegrees=0, swing2LimitDegrees=2, twistLimitDegrees=0}
 {bone1Name=dyn_skirt_sd_04_l, bone2Name=dyn_skirt_sd_03_l, swing1LimitDegrees=0, swing2LimitDegrees=5, twistLimitDegrees=0}
]
