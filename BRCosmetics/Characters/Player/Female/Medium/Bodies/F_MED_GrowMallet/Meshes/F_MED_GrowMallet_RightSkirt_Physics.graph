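# Physics asset for the right skirt of the female medium GrowMallet cosmetic body, covering front and side skirt bone chains.
# Each chain links four capsule bodies in sequence from a tiny base capsule, with tight angular limits that widen at the last joint.
# Side-chain limits allow only swing2, while the front also permits slight swing1 and twist.
# An unconstrained right calf capsule sits alongside a shapeless right thigh body.
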PhysicsAsset /BRCosmetics/Characters/Player/Female/Medium/Bodies/F_MED_GrowMallet/Meshes/F_MED_GrowMallet_RightSkirt_Physics
bodies {
 base_skirt_fr_r [{shapeType=Capsule, length=0.505, radius=0.505, shapeName=base_skirt_fr_r_capsule, rotation=<rot 0 -90 0>}]
 base_skirt_sd_r [{shapeType=Capsule, length=0.505, radius=0.505, shapeName=base_skirt_sd_r_capsule, rotation=<rot 0 -90 0>}]
 calf_r [{
   shapeType=Capsule
   length=33.3659
   radius=7.0424
   shapeName=calf_l_capsule
   center=<21.2656 2.492 0.2506>
   rotation=<rot -0.1636 92.8952 91.8884>
  }]
 dyn_skirt_fr_01_r [{shapeType=Capsule, length=12.4361, radius=5, shapeName=dyn_skirt_fr_01_r_capsule, center=<7 0 5>, rotation=<rot 90 90 0>}]
 dyn_skirt_fr_02_r [{shapeType=Capsule, length=13.9116, radius=6.5, shapeName=dyn_skirt_fr_02_r_capsule, center=<6 0 4>, rotation=<rot 90 90 0>}]
 dyn_skirt_fr_03_r [{shapeType=Capsule, length=13.9116, radius=7, shapeName=dyn_skirt_fr_03_r_capsule, center=<6 0 3>, rotation=<rot 90 90 0>}]
 dyn_skirt_fr_04_r [{shapeType=Capsule, length=13.9116, radius=7, shapeName=dyn_skirt_fr_04_r_capsule, center=<6 0 3>, rotation=<rot 90 90 0>}]
 dyn_skirt_sd_01_r [{shapeType=Capsule, length=6, radius=4.5, shapeName=dyn_skirt_sd_01_r_capsule, center=<6 0 3>, rotation=<rot 90 90 0>}]
 dyn_skirt_sd_02_r [{shapeType=Capsule, length=12.6116, radius=5.75, shapeName=dyn_skirt_sd_02_r_capsule, center=<6 0 3>, rotation=<rot 90 90 0>}]
 dyn_skirt_sd_03_r [{shapeType=Capsule, length=12.6116, radius=6.25, shapeName=dyn_skirt_sd_03_r_capsule, center=<5.3632 0 3>, rotation=<rot 90 90 0>}]
 dyn_skirt_sd_04_r [{shapeType=Capsule, length=12.6116, radius=6.75, shapeName=dyn_skirt_sd_04_r_capsule, center=<6 0 3>, rotation=<rot 90 90 0>}]
 thigh_r []
}
constraints [
 {bone1Name=dyn_skirt_fr_01_r, bone2Name=base_skirt_fr_r, swing1LimitDegrees=3, swing2LimitDegrees=5, twistLimitDegrees=1}
 {bone1Name=dyn_skirt_fr_02_r, bone2Name=dyn_skirt_fr_01_r, swing1LimitDegrees=3, swing2LimitDegrees=5, twistLimitDegrees=1}
 {bone1Name=dyn_skirt_fr_03_r, bone2Name=dyn_skirt_fr_02_r, swing1LimitDegrees=3, swing2LimitDegrees=5, twistLimitDegrees=1}
 {bone1Name=dyn_skirt_fr_04_r, bone2Name=dyn_skirt_fr_03_r, swing1LimitDegrees=3, swing2LimitDegrees=10, twistLimitDegrees=1}
 {bone1Name=dyn_skirt_sd_01_r, bone2Name=base_skirt_sd_r, swing1LimitDegrees=0, swing2LimitDegrees=5, twistLimitDegrees=0}
 {bone1Name=dyn_skirt_sd_02_r, bone2Name=dyn_skirt_sd_01_r, swing1LimitDegrees=0, swing2LimitDegrees=5, twistLimitDegrees=0}
 {bone1Name=dyn_skirt_sd_03_r, bone2Name=dyn_skirt_sd_02_r, swing1LimitDegrees=0, swing2LimitDegrees=5, twistLimitDegrees=0}
 {bone1Name=dyn_skirt_sd_04_r, bone2Name=dyn_skirt_sd_03_r, swing1LimitDegrees=0, swing2LimitDegrees=10, twistLimitDegrees=0}
]
