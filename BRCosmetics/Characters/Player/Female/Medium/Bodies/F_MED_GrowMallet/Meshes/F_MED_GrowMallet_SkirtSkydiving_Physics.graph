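# Physics asset for the GrowMallet female medium body's skirt, named for use while skydiving.
# Capsules on the pelvis, thighs and calves accompany eight skirt chains (front, back, sides), each with a tiny base capsule and four larger dynamic capsules.
# Constraints link only the lower three dynamic bones of each chain, leaving base and first bones unconstrained.
# Twist limits are zero, and back chains swing more freely than the tighter front and sides.
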PhysicsAsset /BRCosmetics/Characters/Player/Female/Medium/Bodies/F_MED_GrowMallet/Meshes/F_MED_GrowMallet_SkirtSkydiving_Physics
bodies {
 base_skirt_bk_c [{shapeType=Capsule, length=1, radius=1, shapeName=base_skirt_bk_c_capsule, center=<6 0 0>, rotation=<rot 0 -90 0>}]
 base_skirt_bk_l [{shapeType=Capsule, length=1, radius=1, shapeName=base_skirt_bk_l_capsule, center=<6 0 0>, rotation=<rot 0 -90 0>}]
 base_skirt_bk_r [{shapeType=Capsule, length=1, radius=1, shapeName=base_skirt_bk_r_capsule, center=<6 0 0>, rotation=<rot 0 -90 0>}]
 base_skirt_fr_c [{shapeType=Capsule, length=1, radius=1, shapeName=base_skirt_fr_c_capsule, center=<6 0 0>, rotation=<rot 0 -90 0>}]
 base_skirt_fr_l [{shapeType=Capsule, length=1, radius=1, shapeName=base_skirt_fr_l_capsule, center=<6 0 0>, rotation=<rot 0 -90 0>}]
 base_skirt_fr_r [{shapeType=Capsule, length=1, radius=1, shapeName=base_skirt_fr_r_capsule, center=<6 0 0>, rotation=<rot 0 -90 0>}]
 base_skirt_sd_l [{shapeType=Capsule, length=1, radius=1, shapeName=base_skirt_sd_l_capsule, center=<6 0 0>, rotation=<rot 90 90 0>}]
 base_skirt_sd_r [{shapeType=Capsule, length=1, radius=1, shapeName=base_skirt_sd_r_capsule, center=<6 0 0>, rotation=<rot 0 -90 0>}]
 calf_l [{
   shapeType=Capsule
   length=32.0754
   radius=9.4798
   shapeName=calf_l_capsule
   center=<-19.8825 -2.4679 -0.469>
   rotation=<rot -0.6342 88.9938 -88.8977>
  }]
 calf_r [{
   shapeType=Capsule
   length=31.379
   radius=10.1051
   shapeName=calf_r_capsule
   center=<21.5278 2.0345 0.5772>
   rotation=<rot -0.6342 88.9938 91.6464>
  }]
 dyn_skirt_bk_01_c [{shapeType=Capsule, length=4, radius=7, shapeName=dyn_skirt_bk_01_c_capsule, center=<8 0 0>, rotation=<rot 90 90 0>}]
 dyn_skirt_bk_01_l [{shapeType=Capsule, length=4, radius=7, shapeName=dyn_skirt_bk_01_l_capsule, center=<8 0 0>, rotation=<rot 90 90 0>}]
 dyn_skirt_bk_01_r [{shapeType=Capsule, length=4, radius=7, shapeName=dyn_skirt_bk_01_r_capsule, center=<8 0 0>, rotation=<rot 90 90 0>}]
 dyn_skirt_bk_02_c [{shapeType=Capsule, length=4, radius=7, shapeName=dyn_skirt_bk_02_c_capsule, center=<8 0 0>, rotation=<rot 90 90 0>}]
 dyn_skirt_bk_02_l [{shapeType=Capsule, length=4, radius=7, shapeName=dyn_skirt_bk_02_l_capsule, center=<8 0 0>, rotation=<rot 90 90 0>}]
 dyn_skirt_bk_02_r [{shapeType=Capsule, length=4, radius=7, shapeName=dyn_skirt_bk_02_r_capsule, center=<8 0 0>, rotation=<rot 90 90 0>}]
 dyn_skirt_bk_03_c [{shapeType=Capsule, length=13.0693, radius=7, shapeName=dyn_skirt_bk_03_c_capsule, center=<8 0 0>, rotation=<rot 90 90 0>}]
 dyn_skirt_bk_03_l [{shapeType=Capsule, length=4, radius=7, shapeName=dyn_skirt_bk_03_l_capsule, center=<8 0 0>, rotation=<rot 90 90 0>}]
 dyn_skirt_bk_03_r [{shapeType=Capsule, length=4, radius=7, shapeName=dyn_skirt_bk_03_r_capsule, center=<8 0 0>, rotation=<rot 90 90 0>}]
 dyn_skirt_bk_04_c [{shapeType=Capsule, length=13.2923, radius=7, shapeName=dyn_skirt_bk_04_c_capsule, center=<11 0 0>, rotation=<rot 90 90 0>}]
 dyn_skirt_bk_04_l [{shapeType=Capsule, length=9.6041, radius=7, shapeName=dyn_skirt_bk_04_l_capsule, center=<11 0 0>, rotation=<rot 90 90 0>}]
 dyn_skirt_bk_04_r [{shapeType=Capsule, length=9.6041, radius=7, shapeName=dyn_skirt_bk_04_r_capsule, center=<11 0 0>, rotation=<rot 90 -90 0>}]
 dyn_skirt_fr_01_c [{shapeType=Capsule, length=4, radius=7, shapeName=dyn_skirt_fr_01_c_capsule, center=<8 0 0>, rotation=<rot 90 90 0>}]
 dyn_skirt_fr_01_l [{shapeType=Capsule, length=4, radius=7, shapeName=dyn_skirt_fr_01_l_capsule, center=<8 0 0>, rotation=<rot 90 90 0>}]
 dyn_skirt_fr_01_r [{shapeType=Capsule, length=4, radius=7, shapeName=dyn_skirt_fr_01_r_capsule, center=<8 0 0>, rotation=<rot 90 90 0>}]
 dyn_skirt_fr_02_c [{shapeType=Capsule, length=4, radius=7, shapeName=dyn_skirt_fr_02_c_capsule, center=<8 0 2>, rotation=<rot 90 90 0>}]
 dyn_skirt_fr_02_l [{shapeType=Capsule, length=4, radius=7, shapeName=dyn_skirt_fr_02_l_capsule, center=<8 0 2>, rotation=<rot 90 90 0>}]
 dyn_skirt_fr_02_r [{shapeType=Capsule, length=4, radius=7, shapeName=dyn_skirt_fr_02_r_capsule, center=<8 0 2>, rotation=<rot 90 90 0>}]
 dyn_skirt_fr_03_c [{shapeType=Capsule, length=4, radius=7, shapeName=dyn_skirt_fr_03_c_capsule, center=<8 0 2>, rotation=<rot 90 90 0>}]
 dyn_skirt_fr_03_l [{shapeType=Capsule, length=4, radius=7, shapeName=dyn_skirt_fr_03_l_capsule, center=<8 0 2>, rotation=<rot 90 90 0>}]
 dyn_skirt_fr_03_r [{shapeType=Capsule, length=4, radius=7, shapeName=dyn_skirt_fr_03_r_capsule, center=<8 0 2>, rotation=<rot 90 90 0>}]
 dyn_skirt_fr_04_c [{shapeType=Capsule, length=4, radius=7, shapeName=dyn_skirt_fr_04_c_capsule, center=<11 0 2.5>, rotation=<rot 90 90 0>}]
 dyn_skirt_fr_04_l [{shapeType=Capsule, length=4, radius=7, shapeName=dyn_skirt_fr_04_l_capsule, center=<11 0 2.5>, rotation=<rot 90 90 0>}]
 dyn_skirt_fr_04_r [{shapeType=Capsule, length=4, radius=7, shapeName=dyn_skirt_fr_04_r_capsule, center=<11 0 2.5>, rotation=<rot 90 90 0>}]
 dyn_skirt_sd_01_l [{shapeType=Capsule, length=4, radius=7, shapeName=dyn_skirt_sd_01_l_capsule, center=<8 0 0>, rotation=<rot 90 90 0>}]
 dyn_skirt_sd_01_r [{shapeType=Capsule, length=4, radius=7, shapeName=dyn_skirt_sd_01_r_capsule, center=<8 0 0>, rotation=<rot 90 90 0>}]
 dyn_skirt_sd_02_l [{shapeType=Capsule, length=4, radius=7, shapeName=dyn_skirt_sd_02_l_capsule, center=<8 0 0>, rotation=<rot 90 90 0>}]
 dyn_skirt_sd_02_r [{shapeType=Capsule, length=4, radius=7, shapeName=dyn_skirt_sd_02_r_capsule, center=<8 0 0>, rotation=<rot 90 90 0>}]
 dyn_skirt_sd_03_l [{shapeType=Capsule, length=4, radius=7, shapeName=dyn_skirt_sd_03_l_capsule, center=<8 0 0>, rotation=<rot 90 90 0>}]
 dyn_skirt_sd_03_r [{shapeType=Capsule, length=4, radius=7, shapeName=dyn_skirt_sd_03_r_capsule, center=<8 0 0>, rotation=<rot 90 90 0>}]
 dyn_skirt_sd_04_l [{shapeType=Capsule, length=4, radius=7, shapeName=dyn_skirt_sd_04_l_capsule, center=<11 0 0>, rotation=<rot 90 90 0>}]
 dyn_skirt_sd_04_r [{shapeType=Capsule, length=4, radius=7, shapeName=dyn_skirt_sd_04_r_capsule, center=<11 0 0>, rotation=<rot 90 90 0>}]
 pelvis [{shapeType=Capsule, length=8.8622, radius=10.886, center=<-2.7959 -3.2735 -0.1981>}]
 thigh_l [{shapeType=Capsule, length=24.2225, radius=11.7776, center=<-26.1476 -2.565 0.6367>, rotation=<rot 90 0.9166 0>}]
 thigh_r [{shapeType=Capsule, length=24.2225, radius=11.7776, center=<26.1476 2.565 -0.6367>, rotation=<rot 90 -179.0834 0>}]
}
constraints [
 {bone1Name=dyn_skirt_bk_03_c, bone2Name=dyn_skirt_bk_02_c, swing1LimitDegrees=8, swing2LimitDegrees=8, twistLimitDegrees=0}
 {bone1Name=dyn_skirt_bk_04_c, bone2Name=dyn_skirt_bk_03_c, swing1LimitDegrees=8, swing2LimitDegrees=8, twistLimitDegrees=0}
 {bone1Name=dyn_skirt_bk_03_l, bone2Name=dyn_skirt_bk_02_l, swing1LimitDegrees=8, swing2LimitDegrees=8, twistLimitDegrees=0}
 {bone1Name=dyn_skirt_bk_04_l, bone2Name=dyn_skirt_bk_03_l, swing1LimitDegrees=8, swing2LimitDegrees=8, twistLimitDegrees=0}
 {bone1Name=dyn_skirt_bk_03_r, bone2Name=dyn_skirt_bk_02_r, swing1LimitDegrees=8, swing2LimitDegrees=8, twistLimitDegrees=0}
 {bone1Name=dyn_skirt_bk_04_r, bone2Name=dyn_skirt_bk_03_r, swing1LimitDegrees=8, swing2LimitDegrees=8, twistLimitDegrees=0}
 {bone1Name=dyn_skirt_fr_03_c, bone2Name=dyn_skirt_fr_02_c, swing1LimitDegrees=2.5, swing2LimitDegrees=2.5, twistLimitDegrees=0}
 {bone1Name=dyn_skirt_fr_04_c, bone2Name=dyn_skirt_fr_03_c, swing1LimitDegrees=2.5, swing2LimitDegrees=2.5, twistLimitDegrees=0}
 {bone1Name=dyn_skirt_fr_03_l, bone2Name=dyn_skirt_fr_02_l, swing1LimitDegrees=2.5, swing2LimitDegrees=2.5, twistLimitDegrees=0}
 {bone1Name=dyn_skirt_fr_04_l, bone2Name=dyn_skirt_fr_03_l, swing1LimitDegrees=2.5, swing2LimitDegrees=2.5, twistLimitDegrees=0}
 {bone1Name=dyn_skirt_fr_03_r, bone2Name=dyn_skirt_fr_02_r, swing1LimitDegrees=2.5, swing2LimitDegrees=2.5, twistLimitDegrees=0}
 {bone1Name=dyn_skirt_fr_04_r, bone2Name=dyn_skirt_fr_03_r, swing1LimitDegrees=2.5, swing2LimitDegrees=2.5, twistLimitDegrees=0}
 {bone1Name=dyn_skirt_sd_03_l, bone2Name=dyn_skirt_sd_02_l, swing1LimitDegrees=2.5, swing2LimitDegrees=2.5, twistLimitDegrees=0}
 {bone1Name=dyn_skirt_sd_04_l, bone2Name=dyn_skirt_sd_03_l, swing1LimitDegrees=2.5, swing2LimitDegrees=2.5, twistLimitDegrees=0}
 {bone1Name=dyn_skirt_sd_03_r, bone2Name=dyn_skirt_sd_02_r, swing1LimitDegrees=2.5, swing2LimitDegrees=2.5, twistLimitDegrees=0}
 {bone1Name=dyn_skirt_sd_04_r, bone2Name=dyn_skirt_sd_03_r, swing1LimitDegrees=2.5, swing2LimitDegrees=2.5, twistLimitDegrees=0}
]
