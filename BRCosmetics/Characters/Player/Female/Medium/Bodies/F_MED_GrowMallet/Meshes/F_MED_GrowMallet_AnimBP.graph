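# Post-process style AnimBP for the F_MED_GrowMallet female body cosmetic, adding secondary skirt, necklace and chest motion to the pose copied from the parent mesh.
# A Control Rig corrects that pose from movement states, then feeds RigidBody skirt simulations, switching between per-thigh and pelvis-based skydiving setups.
# Anim Dynamics, weighted bone constraints and additive bone offsets refine skirt and necklace.
# Cached poses link the stages; bPbBlendEnabled toggles the final pec-bounce layer.
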AnimBlueprint /BRCosmetics/Characters/Player/Female/Medium/Bodies/F_MED_GrowMallet/Meshes/F_MED_GrowMallet_AnimBP
target_skeleton=&/F_MED_GrowMallet_Skeleton
graphs [
 AnimGraph {nodes [
   "Constraint - base_skirt_fr_c (PP)"!(ComponentPose, Alpha=1, ConstraintWeights_0=0.5, ConstraintWeights_1:float=0.5) -> (Pose) {
    alphaScaleBiasClamp {bInterpResult=true, interpSpeedDecreasing=5, interpSpeedIncreasing=3.5}
    boneToModify {boneName=base_skirt_fr_c}
    constraintSetup [{targetBone {boneName=base_skirt_fr_l}}, {targetBone {boneName=base_skirt_fr_r}}]
    constraintWeights [1,1]
   }
   n1 = "Constraint - base_skirt_fr_l (PPPPPPPP)"!(
    ComponentPose
    Alpha=1
    ConstraintWeights_0=0.5131
    ConstraintWeights_1:float=0.3076
    ConstraintWeights_2:float=0.0587
    ConstraintWeights_3:float=0.0548
    ConstraintWeights_4:float=0.0525
    ConstraintWeights_5:float=0.0058
    ConstraintWeights_6:float=0.0041
    ConstraintWeights_7:float=0.0034
   ) -> (Pose) {
    boneToModify {boneName=base_skirt_fr_l}
    constraintSetup [
     {targetBone {boneName=pelvis}}
     {targetBone {boneName=thigh_l}}
     {targetBone {boneName=dyn_skirt_fr_01_c}}
     {targetBone {boneName=spine_02}}
     {targetBone {boneName=thigh_twist_01_l}}
     {targetBone {boneName=thigh_r}}
     {targetBone {boneName=dyn_skirt_fr_02_c}}
     {targetBone {boneName=spine_04}}
    ]
    constraintWeights [1,1,1,1,1,1,1,1]
   }
   n2 = "Constraint - base_skirt_sd_l (PPPPPPPP)"!(
    ComponentPose=@n1
    Alpha=1
    ConstraintWeights_0=0.5473
    ConstraintWeights_1:float=0.2454
    ConstraintWeights_2:float=0.071
    ConstraintWeights_3:float=0.0636
    ConstraintWeights_4:float=0.0554
    ConstraintWeights_5:float=0.011
    ConstraintWeights_6:float=0.0032
    ConstraintWeights_7:float=0.0029
   ) -> (Pose) {
    boneToModify {boneName=base_skirt_sd_l}
    constraintSetup [
     {targetBone {boneName=pelvis}}
     {targetBone {boneName=thigh_l}}
     {targetBone {boneName=dyn_skirt_bk_01_l}}
     {targetBone {boneName=thigh_twist_01_l}}
     {targetBone {boneName=spine_02}}
     {targetBone {boneName=dyn_skirt_bk_02_l}}
     {targetBone {boneName=spine_01}}
     {targetBone {boneName=spine_04}}
    ]
    constraintWeights [1,1,1,1,1,1,1,1]
   }
   n3 = "Constraint - base_skirt_bk_l (PPPPPPPP)"!(
    ComponentPose=@n2
    Alpha=1
    ConstraintWeights_0=0.5053
    ConstraintWeights_1:float=0.1784
    ConstraintWeights_2:float=0.1242
    ConstraintWeights_3:float=0.072
    ConstraintWeights_4:float=0.0544
    ConstraintWeights_5:float=0.0427
    ConstraintWeights_6:float=0.0123
    ConstraintWeights_7:float=0.0106
   ) -> (Pose) {
    boneToModify {boneName=base_skirt_bk_l}
    constraintSetup [
     {targetBone {boneName=pelvis}}
     {targetBone {boneName=dyn_skirt_bk_01_l}}
     {targetBone {boneName=thigh_l}}
     {targetBone {boneName=dyn_skirt_bk_01_c}}
     {targetBone {boneName=dyn_skirt_bk_02_l}}
     {targetBone {boneName=spine_02}}
     {targetBone {boneName=dyn_skirt_bk_02_c}}
     {targetBone {boneName=thigh_twist_01_l}}
    ]
    constraintWeights [1,1,1,1,1,1,1,1]
   }
   n4 = "Constraint - base_skirt_fr_c (PPPPPPPP)"!(
    ComponentPose=@n3
    Alpha=1
    ConstraintWeights_0=0.5003
    ConstraintWeights_1:float=0.1277
    ConstraintWeights_2:float=0.1268
    ConstraintWeights_3:float=0.1262
    ConstraintWeights_4:float=0.0801
    ConstraintWeights_5:float=0.0223
    ConstraintWeights_6:float=0.0109
    ConstraintWeights_7:float=0.0057
   ) -> (Pose) {
    boneToModify {boneName=base_skirt_fr_c}
    constraintSetup [
     {targetBone {boneName=pelvis}}
     {targetBone {boneName=dyn_skirt_fr_01_c}}
     {targetBone {boneName=thigh_r}}
     {targetBone {boneName=thigh_l}}
     {targetBone {boneName=spine_02}}
     {targetBone {boneName=dyn_skirt_fr_02_c}}
     {targetBone {boneName=spine_04}}
     {targetBone {boneName=thigh_twist_01_r}}
    ]
    constraintWeights [1,1,1,1,1,1,1,1]
   }
   n5 = "Constraint - base_skirt_bk_c (PPPPPPPP)"!(
    ComponentPose=@n4
    Alpha=1
    ConstraintWeights_0=0.5539
    ConstraintWeights_1:float=0.1278
    ConstraintWeights_2:float=0.1041
    ConstraintWeights_3:float=0.0978
    ConstraintWeights_4:float=0.055
    ConstraintWeights_5:float=0.029
    ConstraintWeights_6:float=0.0176
    ConstraintWeights_7:float=0.0148
   ) -> (Pose) {
    boneToModify {boneName=base_skirt_bk_c}
    constraintSetup [
     {targetBone {boneName=pelvis}}
     {targetBone {boneName=dyn_skirt_bk_01_c}}
     {targetBone {boneName=dyn_skirt_bk_01_r}}
     {targetBone {boneName=dyn_skirt_bk_01_l}}
     {targetBone {boneName=spine_02}}
     {targetBone {boneName=dyn_skirt_bk_02_c}}
     {targetBone {boneName=dyn_skirt_bk_02_r}}
     {targetBone {boneName=dyn_skirt_bk_02_l}}
    ]
    constraintWeights [1,1,1,1,1,1,1,1]
   }
   n6 = "Constraint - base_skirt_fr_r (PPPPPPPP)"!(
    ComponentPose=@n5
    Alpha=1
    ConstraintWeights_0=0.5164
    ConstraintWeights_1:float=0.3049
    ConstraintWeights_2:float=0.0582
    ConstraintWeights_3:float=0.0541
    ConstraintWeights_4:float=0.054
    ConstraintWeights_5:float=0.0056
    ConstraintWeights_6:float=0.0037
    ConstraintWeights_7:float=0.0031
   ) -> (Pose) {
    boneToModify {boneName=base_skirt_fr_r}
    constraintSetup [
     {targetBone {boneName=pelvis}}
     {targetBone {boneName=thigh_r}}
     {targetBone {boneName=dyn_skirt_fr_01_c}}
     {targetBone {boneName=spine_02}}
     {targetBone {boneName=thigh_twist_01_r}}
     {targetBone {boneName=thigh_l}}
     {targetBone {boneName=dyn_skirt_fr_02_c}}
     {targetBone {boneName=spine_04}}
    ]
    constraintWeights [1,1,1,1,1,1,1,1]
   }
   n7 = "Constraint - base_skirt_sd_r (PPPPPPPP)"!(
    ComponentPose=@n6
    Alpha=1
    ConstraintWeights_0=0.5526
    ConstraintWeights_1:float=0.2423
    ConstraintWeights_2:float=0.0716
    ConstraintWeights_3:float=0.0661
    ConstraintWeights_4:float=0.0534
    ConstraintWeights_5:float=0.0094
    ConstraintWeights_6:float=0.0032
    ConstraintWeights_7:float=0.0014
   ) -> (Pose) {
    boneToModify {boneName=base_skirt_sd_r}
    constraintSetup [
     {targetBone {boneName=pelvis}}
     {targetBone {boneName=thigh_r}}
     {targetBone {boneName=dyn_skirt_bk_01_r}}
     {targetBone {boneName=thigh_twist_01_r}}
     {targetBone {boneName=spine_02}}
     {targetBone {boneName=dyn_skirt_bk_02_r}}
     {targetBone {boneName=spine_01}}
     {targetBone {boneName=dyn_skirt_bk_01_c}}
    ]
    constraintWeights [1,1,1,1,1,1,1,1]
   }
   "Constraint - base_skirt_bk_r (PPPPPPPP)"!(
    ComponentPose=@n7
    Alpha=1
    ConstraintWeights_0=0.5196
    ConstraintWeights_1:float=0.1799
    ConstraintWeights_2:float=0.113
    ConstraintWeights_3:float=0.0723
    ConstraintWeights_4:float=0.0494
    ConstraintWeights_5:float=0.036
    ConstraintWeights_6:float=0.022
    ConstraintWeights_7:float=0.0078
   ) -> (Pose) {
    boneToModify {boneName=base_skirt_bk_r}
    constraintSetup [
     {targetBone {boneName=pelvis}}
     {targetBone {boneName=dyn_skirt_bk_01_r}}
     {targetBone {boneName=thigh_r}}
     {targetBone {boneName=dyn_skirt_bk_01_c}}
     {targetBone {boneName=dyn_skirt_bk_02_r}}
     {targetBone {boneName=spine_02}}
     {targetBone {boneName=thigh_twist_01_r}}
     {targetBone {boneName=dyn_skirt_bk_02_c}}
    ]
    constraintWeights [1,1,1,1,1,1,1,1]
   }
   Output Pose(Result=Blend Poses by bool(
     BlendPose_0=Use cached pose 'PostBodyDynamics'()
     BlendPose_1=Use cached pose 'PostAnimDynamics'()
     BlendTime_0=0
     BlendTime_1=0
     bActiveValue=Get bPbBlendEnabled() -> (bPbBlendEnabled:bool)
    )) {layerGroup=None}
   n8 = Control Rig!(
    Source=Copy Pose From Mesh() {bUseAttachedParent=true}
    Alpha
    "isCrouching ":double=Get Crouching() -> (Crouching:double)
    isCrouchingMoving:double=Get Crouching_sideWays() -> (Crouching_sideWays:double)
    "isMoving_Forward ":double=Get moveForward() -> (moveForward:double)
    isMoving_Left:double=Get strafeLeft float() -> (strafeLeft float:double)
    isMoving_Right:double=Get strafeRight float() -> (strafeRight float:double)
    isMoving_Back:double=Get moveBackward() -> (moveBackward:double)
    SpritingRunning:double=Get Sprinting_Running() -> (Sprinting_Running:double)
    isJumping:double=Get isJumping() -> (isJumping:double)
    swingPickaxe:bool=Get bIsPlayingMeleeAnim()
    isParachuteOpened:double=Get Is_Parachute_Open() -> (Is_Parachute_Open:double)
    "Crouching_Sideways ":double=Get Crouching_sideWays()
    isPlayingEmote:bool=Get bIsPlayingEmote()
    RegularCrouching:double=Get RegularCrouching() -> (RegularCrouching:double)
    SkyDive_Floating:double=Get SkyDiving_Alpha() -> (SkyDiving_Alpha:double)
    isOnAir:double=Get Is_on_Air() -> (Is_on_Air:double)
    SkyDiving:double=Get SkyDiving_Alpha()
    GrindRail:double=Get isGrindRail() -> (isGrindRail:double)
    Skirt_Fixes:double=Get SkirtPostAlpha() -> (SkirtPostAlpha:double)
    DBNO Carried:bool=Get bIsDBNOCarried()
   ) -> (Pose) {controlRigAssetReference {blueprintRigClass=&/F_MED_GrowMallet_CtrlRig.F_MED_GrowMallet_CtrlRig_C}}
   Save cached pose 'ControlRig Cache'(Pose=@n8) {cacheName=ControlRig Cache}
   n9 = Anim Dynamics - Bone: Necklace_dyn_01(ComponentPose=Local To Component(LocalPose=Use cached pose 'Initial Cache Pose'()), Alpha=1, ExternalForce="0,0,0") {
    angularDampingOverride=1
    angularSpringConstant=15000
    bAngularSpring=true
    bOverrideAngularDamping=true
    bUseGravityOverride=true
    gravityScale=0.35
    simulationSpace=RootRelative
    boundBone {boneName=Necklace_dyn_01}
    componentAppliedLinearAccClamp=<4000 4000 4000>
    componentLinearAccScale=<0.4 0.4 0.4>
    componentLinearVelScale=<0.4 0.4 0.4>
    gravityOverride {z=-980}
    physicsBodyDefinitions [{
      boundBone {boneName=Necklace_dyn_01}
      boxExtents=<25 35 35>
      constraintSetup {angularLimitsMax=<0 15 20>, angularLimitsMin=<0 -15 0>, angularTarget=<1 0 0>}
      localJointOffset=<4 0 0>
     }]
    simSpaceSettings {maxAngularAcceleration=100, maxAngularVelocity=10, simSpaceAngularAlpha=1}
   }
   n10 = "Constraint - dyn_skirt_bk_01_c (UU)"!(ComponentPose=@n9, Alpha=1, ConstraintWeights_0=0.5, ConstraintWeights_1:float=0.5) -> (Pose) {
    boneToModify {boneName=dyn_skirt_bk_01_c}
    constraintSetup [{transformType=LookAt, targetBone {boneName=dyn_skirt_bk_01_l}}, {transformType=LookAt, targetBone {boneName=dyn_skirt_bk_01_r}}]
    constraintWeights [1,1]
   }
   n11 = "Constraint - dyn_skirt_bk_02_c (UU)"!(ComponentPose=@n10, Alpha=1, ConstraintWeights_0=0.5, ConstraintWeights_1:float=0.5) -> (Pose) {
    boneToModify {boneName=dyn_skirt_bk_02_c}
    constraintSetup [{transformType=LookAt, targetBone {boneName=dyn_skirt_bk_02_l}}, {transformType=LookAt, targetBone {boneName=dyn_skirt_bk_02_r}}]
    constraintWeights [1,1]
   }
   n12 = "Constraint - dyn_skirt_bk_03_c (UU)"!(ComponentPose=@n11, Alpha=1, ConstraintWeights_0=0.5, ConstraintWeights_1:float=0.5) -> (Pose) {
    boneToModify {boneName=dyn_skirt_bk_03_c}
    constraintSetup [{transformType=LookAt, targetBone {boneName=dyn_skirt_bk_03_l}}, {transformType=LookAt, targetBone {boneName=dyn_skirt_bk_03_r}}]
    constraintWeights [1,1]
   }
   n13 = "Constraint - dyn_skirt_bk_04_c (UU)"!(ComponentPose=@n12, Alpha=1, ConstraintWeights_0=0.5, ConstraintWeights_1:float=0.5) -> (Pose) {
    boneToModify {boneName=dyn_skirt_bk_04_c}
    constraintSetup [{transformType=LookAt, targetBone {boneName=dyn_skirt_bk_04_l}}, {transformType=LookAt, targetBone {boneName=dyn_skirt_bk_04_r}}]
    constraintWeights [1,1]
   }
   n14 = "Constraint - base_skirt_fr_c (UU)"!(ComponentPose=@n13, Alpha=1, ConstraintWeights_0=0.5, ConstraintWeights_1:float=0.5) -> (Pose) {
    alphaScaleBiasClamp {bInterpResult=true, interpSpeedDecreasing=5, interpSpeedIncreasing=3.5}
    boneToModify {boneName=base_skirt_fr_c}
    constraintSetup [{transformType=LookAt, targetBone {boneName=base_skirt_fr_l}}, {transformType=LookAt, targetBone {boneName=base_skirt_fr_r}}]
    constraintWeights [1,1]
   }
   n15 = "Transform (Modify) Bone - Bone: base_skirt_fr_r"(ComponentPose=@n14, Alpha=1, Translation="0,0,0", Rotation="4,-5,0", Scale="1,1,1") {
    rotationMode=BMM_Additive
    rotationSpace=BCS_BoneSpace
    translationMode=BMM_Additive
    alphaScaleBiasClamp {bInterpResult=true, interpSpeedDecreasing=5, interpSpeedIncreasing=5}
    boneToModify {boneName=base_skirt_fr_r}
    rotation {pitch=4, yaw=-5}
   }
   n16 = Anim Dynamics - Bone: dyn_skirt_sd_04_l!(ComponentPose=@n15, Alpha=1, GravityOverride:Vector="0,0,-980", ExternalForce="0,0,0") -> (Pose) {
    angularDampingOverride=1
    angularSpringConstant=5000
    bAngularSpring=true
    bOverrideAngularDamping=true
    bUseGravityOverride=true
    gravityScale=0.35
    simulationSpace=RootRelative
    boundBone {boneName=dyn_skirt_sd_04_l}
    componentAppliedLinearAccClamp=<3000 3000 3000>
    componentLinearAccScale=<0.3 0.3 0.3>
    componentLinearVelScale=<0.3 0.3 0.3>
    gravityOverride {z=-980}
    physicsBodyDefinitions [{
      boundBone {boneName=dyn_skirt_sd_04_l}
      boxExtents=<15 20 20>
      constraintSetup {angularLimitsMax=<0 0 4>, angularLimitsMin=<0 -15 -20>, angularTarget=<1 0 0>}
      localJointOffset=<4 0 0>
     }]
    simSpaceSettings {maxAngularAcceleration=100, maxAngularVelocity=10, simSpaceAngularAlpha=1}
   }
   n17 = Anim Dynamics - Bone: dyn_skirt_bk_04_l!(ComponentPose=@n16, Alpha=1, GravityOverride:Vector="0,0,-980", ExternalForce="0,0,0") -> (Pose) {
    angularDampingOverride=1
    angularSpringConstant=5000
    bAngularSpring=true
    bOverrideAngularDamping=true
    bUseGravityOverride=true
    gravityScale=0.35
    simulationSpace=RootRelative
    boundBone {boneName=dyn_skirt_bk_04_l}
    componentAppliedLinearAccClamp=<3000 3000 3000>
    componentLinearAccScale=<0.3 0.3 0.3>
    componentLinearVelScale=<0.3 0.3 0.3>
    gravityOverride {z=-980}
    physicsBodyDefinitions [{
      boundBone {boneName=dyn_skirt_bk_04_l}
      boxExtents=<15 20 20>
      constraintSetup {angularLimitsMax=<0 0 4>, angularLimitsMin=<0 -15 -4>, angularTarget=<1 0 0>}
      localJointOffset=<4 0 0>
     }]
    simSpaceSettings {maxAngularAcceleration=100, maxAngularVelocity=10, simSpaceAngularAlpha=1}
   }
   n18 = Anim Dynamics - Bone: dyn_skirt_bk_04_c!(ComponentPose=@n17, Alpha=1, GravityOverride:Vector="0,0,-980", ExternalForce="0,0,0") -> (Pose) {
    angularDampingOverride=1
    angularSpringConstant=5000
    bAngularSpring=true
    bOverrideAngularDamping=true
    bUseGravityOverride=true
    gravityScale=0.35
    simulationSpace=RootRelative
    boundBone {boneName=dyn_skirt_bk_04_c}
    componentAppliedLinearAccClamp=<3000 3000 3000>
    componentLinearAccScale=<0.3 0.3 0.3>
    componentLinearVelScale=<0.3 0.3 0.3>
    gravityOverride {z=-980}
    physicsBodyDefinitions [{
      boundBone {boneName=dyn_skirt_bk_04_c}
      boxExtents=<15 20 20>
      constraintSetup {angularLimitsMax=<0 0 4>, angularLimitsMin=<0 -15 -4>, angularTarget=<1 0 0>}
      localJointOffset=<4 0 0>
     }]
    simSpaceSettings {maxAngularAcceleration=100, maxAngularVelocity=10, simSpaceAngularAlpha=1}
   }
   n19 = Anim Dynamics - Bone: dyn_skirt_bk_04_r!(ComponentPose=@n18, Alpha=1, GravityOverride:Vector="0,0,-980", ExternalForce="0,0,0") -> (Pose) {
    angularDampingOverride=1
    angularSpringConstant=5000
    bAngularSpring=true
    bOverrideAngularDamping=true
    bUseGravityOverride=true
    gravityScale=0.35
    simulationSpace=RootRelative
    boundBone {boneName=dyn_skirt_bk_04_r}
    componentAppliedLinearAccClamp=<3000 3000 3000>
    componentLinearAccScale=<0.3 0.3 0.3>
    componentLinearVelScale=<0.3 0.3 0.3>
    gravityOverride {z=-980}
    physicsBodyDefinitions [{
      boundBone {boneName=dyn_skirt_bk_04_r}
      boxExtents=<15 20 20>
      constraintSetup {angularLimitsMax=<0 0 4>, angularLimitsMin=<0 -15 -4>, angularTarget=<1 0 0>}
      localJointOffset=<4 0 0>
     }]
    simSpaceSettings {maxAngularAcceleration=100, maxAngularVelocity=10, simSpaceAngularAlpha=1}
   }
   n20 = Anim Dynamics - Bone: dyn_skirt_sd_04_r!(ComponentPose=@n19, Alpha=1, GravityOverride:Vector="0,0,-980", ExternalForce="0,0,0") -> (Pose) {
    angularDampingOverride=1
    angularSpringConstant=5000
    bAngularSpring=true
    bOverrideAngularDamping=true
    bUseGravityOverride=true
    gravityScale=0.35
    simulationSpace=RootRelative
    boundBone {boneName=dyn_skirt_sd_04_r}
    componentAppliedLinearAccClamp=<3000 3000 3000>
    componentLinearAccScale=<0.3 0.3 0.3>
    componentLinearVelScale=<0.3 0.3 0.3>
    gravityOverride {z=-980}
    physicsBodyDefinitions [{
      boundBone {boneName=dyn_skirt_sd_04_r}
      boxExtents=<15 20 20>
      constraintSetup {angularLimitsMax=<0 0 25>, angularLimitsMin=<0 -15 -4>, angularTarget=<1 0 0>}
      localJointOffset=<4 0 0>
     }]
    simSpaceSettings {maxAngularAcceleration=100, maxAngularVelocity=10, simSpaceAngularAlpha=1}
   }
   n21 = "Transform (Modify) Bone - Bone: base_skirt_fr_c"(ComponentPose=@n20, Alpha=1, Translation="0,-3,0", Rotation="-3.593,1.871,28.697", Scale="1,1,1") {
    rotationMode=BMM_Additive
    scaleMode=BMM_Additive
    translationMode=BMM_Additive
    boneToModify {boneName=base_skirt_fr_c}
    rotation=<rot -3.593 1.871 28.697>
    translation {y=-3}
   }
   n22 = "Transform (Modify) Bone - Bone: base_skirt_fr_r"(ComponentPose=@n21, Alpha=1, Translation="0,-0.001,3.5", Rotation="0,-4,0", Scale="1,1,1") {
    rotationMode=BMM_Additive
    rotationSpace=BCS_BoneSpace
    translationMode=BMM_Additive
    translationSpace=BCS_BoneSpace
    boneToModify {boneName=base_skirt_fr_r}
    rotation {yaw=-4}
    translation {y=-0.001, z=3.5}
   }
   n23 = "Transform (Modify) Bone - Bone: base_skirt_fr_l"(ComponentPose=@n22, Alpha=1, Translation="4.5,0,1.298", Rotation="0.995,5.154,20.593", Scale="1,1,1") {
    rotationMode=BMM_Additive
    rotationSpace=BCS_BoneSpace
    translationMode=BMM_Additive
    translationSpace=BCS_BoneSpace
    boneToModify {boneName=base_skirt_fr_l}
    rotation=<rot 0.995 5.154 20.593>
    translation {x=4.5, z=1.298}
   }
   n24 = Anim Dynamics - Bone: dyn_skirt_fr_03_l!(ComponentPose=@n23, Alpha=1, GravityOverride:Vector="0,0,-980", ExternalForce="0,0,0") -> (Pose) {
    angularDampingOverride=1
    angularSpringConstant=10000
    bAngularSpring=true
    bOverrideAngularDamping=true
    bUseGravityOverride=true
    gravityScale=0.35
    simulationSpace=RootRelative
    boundBone {boneName=dyn_skirt_fr_03_l}
    componentAppliedLinearAccClamp=<3000 3000 3000>
    componentLinearAccScale=<0.3 0.3 0.3>
    componentLinearVelScale=<0.3 0.3 0.3>
    gravityOverride {z=-980}
    physicsBodyDefinitions [{
      boundBone {boneName=dyn_skirt_fr_03_l}
      boxExtents=<15 15 15>
      constraintSetup {angularLimitsMax=<5 10 0>, angularLimitsMin=<-5 -2 0>, angularTarget=<1 0 0>}
      localJointOffset=<4 0 0>
     }]
    simSpaceSettings {maxAngularAcceleration=100, maxAngularVelocity=10, simSpaceAngularAlpha=1}
   }
   n25 = Anim Dynamics - Bone: dyn_skirt_fr_04_l!(ComponentPose=@n24, Alpha=1, GravityOverride:Vector="0,0,-980", ExternalForce="0,0,0") -> (Pose) {
    angularDampingOverride=1
    angularSpringConstant=8000
    bAngularSpring=true
    bOverrideAngularDamping=true
    bUseGravityOverride=true
    gravityScale=0.35
    simulationSpace=RootRelative
    boundBone {boneName=dyn_skirt_fr_04_l}
    componentAppliedLinearAccClamp=<3000 3000 3000>
    componentLinearAccScale=<0.3 0.3 0.3>
    componentLinearVelScale=<0.3 0.3 0.3>
    gravityOverride {z=-980}
    physicsBodyDefinitions [{
      boundBone {boneName=dyn_skirt_fr_04_l}
      boxExtents=<15 15 15>
      constraintSetup {angularLimitsMax=<20 45 0>, angularLimitsMin=<-20 -5 0>, angularTarget=<1 0 0>}
      localJointOffset=<4 0 0>
     }]
    simSpaceSettings {maxAngularAcceleration=100, maxAngularVelocity=10, simSpaceAngularAlpha=1}
   }
   n26 = Anim Dynamics - Bone: dyn_skirt_fr_03_c!(ComponentPose=@n25, Alpha=1, GravityOverride:Vector="0,0,-980", ExternalForce="0,0,0") -> (Pose) {
    angularDampingOverride=1
    angularSpringConstant=7000
    bAngularSpring=true
    bOverrideAngularDamping=true
    bUseGravityOverride=true
    gravityScale=0.35
    simulationSpace=RootRelative
    boundBone {boneName=dyn_skirt_fr_03_c}
    componentAppliedLinearAccClamp=<3000 3000 3000>
    componentLinearAccScale=<0.3 0.3 0.3>
    componentLinearVelScale=<0.3 0.3 0.3>
    gravityOverride {z=-980}
    physicsBodyDefinitions [{
      boundBone {boneName=dyn_skirt_fr_03_c}
      boxExtents=<15 15 15>
      constraintSetup {angularLimitsMax=<0 10 5>, angularLimitsMin=<0 0 -5>, angularTarget=<1 0 0>}
      localJointOffset=<4 0 0>
     }]
    simSpaceSettings {maxAngularAcceleration=100, maxAngularVelocity=10, simSpaceAngularAlpha=1}
   }
   n27 = Anim Dynamics - Bone: dyn_skirt_fr_04_c!(ComponentPose=@n26, Alpha=1, GravityOverride:Vector="0,0,-980", ExternalForce="0,0,0") -> (Pose) {
    angularDampingOverride=1
    angularSpringConstant=8000
    bAngularSpring=true
    bOverrideAngularDamping=true
    bUseGravityOverride=true
    gravityScale=0.35
    simulationSpace=RootRelative
    boundBone {boneName=dyn_skirt_fr_04_c}
    componentAppliedLinearAccClamp=<3000 3000 3000>
    componentLinearAccScale=<0.3 0.3 0.3>
    componentLinearVelScale=<0.3 0.3 0.3>
    gravityOverride {z=-980}
    physicsBodyDefinitions [{
      boundBone {boneName=dyn_skirt_fr_04_c}
      boxExtents=<15 15 15>
      constraintSetup {angularLimitsMax=<10 40 5>, angularLimitsMin=<-10 0 -5>, angularTarget=<1 0 0>}
      localJointOffset=<4 0 0>
     }]
    simSpaceSettings {maxAngularAcceleration=100, maxAngularVelocity=10, simSpaceAngularAlpha=1}
   }
   n28 = Anim Dynamics - Bone: dyn_skirt_fr_03_r!(ComponentPose=@n27, Alpha=1, GravityOverride:Vector="0,0,-980", ExternalForce="0,0,0") -> (Pose) {
    angularDampingOverride=1
    angularSpringConstant=8000
    bAngularSpring=true
    bOverrideAngularDamping=true
    bUseGravityOverride=true
    gravityScale=0.35
    simulationSpace=RootRelative
    boundBone {boneName=dyn_skirt_fr_03_r}
    componentAppliedLinearAccClamp=<3000 3000 3000>
    componentLinearAccScale=<0.3 0.3 0.3>
    componentLinearVelScale=<0.3 0.3 0.3>
    gravityOverride {z=-980}
    physicsBodyDefinitions [{
      boundBone {boneName=dyn_skirt_fr_03_r}
      boxExtents=<15 15 15>
      constraintSetup {angularLimitsMax=<0 15 4>, angularLimitsMin=<0 -5 -4>, angularTarget=<1 0 0>}
      localJointOffset=<4 0 0>
     }]
    simSpaceSettings {maxAngularAcceleration=100, maxAngularVelocity=10, simSpaceAngularAlpha=1}
   }
   n29 = Anim Dynamics - Bone: dyn_skirt_fr_04_r!(ComponentPose=@n28, Alpha=1, GravityOverride:Vector="0,0,-980", ExternalForce="0,0,0") -> (Pose) {
    angularDampingOverride=1
    angularSpringConstant=8000
    bAngularSpring=true
    bOverrideAngularDamping=true
    bUseGravityOverride=true
    gravityScale=0.35
    simulationSpace=RootRelative
    boundBone {boneName=dyn_skirt_fr_04_r}
    componentAppliedLinearAccClamp=<3000 3000 3000>
    componentLinearAccScale=<0.3 0.3 0.3>
    componentLinearVelScale=<0.3 0.3 0.3>
    gravityOverride {z=-980}
    physicsBodyDefinitions [{
      boundBone {boneName=dyn_skirt_fr_04_r}
      boxExtents=<15 15 15>
      constraintSetup {angularLimitsMax=<0 40 4>, angularLimitsMin=<0 0 -4>, angularTarget=<1 0 0>}
      localJointOffset=<4 0 0>
     }]
    simSpaceSettings {maxAngularAcceleration=100, maxAngularVelocity=10, simSpaceAngularAlpha=1}
   }
   n30 = Anim Dynamics - Bone: dyn_skirt_sd_04_r(ComponentPose=@n29, Alpha=1, GravityOverride="0,0,-980", ExternalForce="0,0,0") {
    angularDampingOverride=1
    angularSpringConstant=7000
    bAngularSpring=true
    bOverrideAngularDamping=true
    bUseGravityOverride=true
    gravityScale=0.35
    simulationSpace=RootRelative
    boundBone {boneName=dyn_skirt_sd_04_r}
    componentAppliedLinearAccClamp=<3000 3000 3000>
    componentLinearAccScale=<0.3 0.3 0.3>
    componentLinearVelScale=<0.3 0.3 0.3>
    gravityOverride {z=-980}
    physicsBodyDefinitions [{
      boundBone {boneName=dyn_skirt_sd_04_r}
      boxExtents=<15 15 15>
      constraintSetup {angularLimitsMax=<0 4 40>, angularLimitsMin=<0 -4 0>, angularTarget=<1 0 0>}
      localJointOffset=<4 0 0>
     }]
    simSpaceSettings {maxAngularAcceleration=100, maxAngularVelocity=10, simSpaceAngularAlpha=1}
   }
   n31 = "Transform (Modify) Bone - Bone: base_skirt_fr_c"(
    ComponentPose="Transform (Modify) Bone - Bone: base_skirt_bk_c"(ComponentPose=@n30, Alpha=1, Translation="0,0,0", Rotation="5,0,5", Scale="1,1,1") {rotationMode=BMM_Additive, boneToModify {boneName=base_skirt_bk_c}, rotation {pitch=5, roll=5}}
    Alpha=1
    Translation="0,0,5"
    Rotation="0,0,-15"
    Scale="1,1,1"
   ) {rotationMode=BMM_Additive, translationMode=BMM_Additive, boneToModify {boneName=base_skirt_fr_c}, rotation {roll=-15}, translation {z=5}}
   n32 = "Transform (Modify) Bone - Bone: base_skirt_bk_l"(ComponentPose=@n31, Alpha=1, Translation="0,0,0", Rotation="5,0,8", Scale="1,1,1") {rotationMode=BMM_Additive, boneToModify {boneName=base_skirt_bk_l}, rotation {pitch=5, roll=8}}
   n33 = "Transform (Modify) Bone - Bone: base_skirt_fr_c"!(
    ComponentPose="Transform (Modify) Bone - Bone: base_skirt_sd_r"(
     ComponentPose="Transform (Modify) Bone - Bone: base_skirt_sd_l"(ComponentPose=@n32, Alpha=1, Translation="0,0,0", Rotation="0,0,-10", Scale="1,1,1") {rotationMode=BMM_Additive, boneToModify {boneName=base_skirt_sd_l}, rotation {roll=-10}}
     Alpha=1
     Translation="0,0,0"
     Rotation="0,0,-10"
     Scale="1,1,1"
    ) {rotationMode=BMM_Additive, boneToModify {boneName=base_skirt_sd_r}, rotation {roll=-10}}
    bAlphaBoolEnabled:bool=True
    Translation="0,0,0"
    Rotation="10,0,0"
    Scale="1,1,1"
   ) -> (Pose) {
    alphaInputType=Bool
    rotationMode=BMM_Additive
    rotationSpace=BCS_BoneSpace
    translationMode=BMM_Additive
    translationSpace=BCS_BoneSpace
    boneToModify {boneName=base_skirt_fr_c}
    rotation {pitch=10}
   }
   n34 = "Transform (Modify) Bone - Bone: base_skirt_bk_l"(ComponentPose=@n33, Alpha=1, Translation="-4,0,0", Rotation="12,-10,0.015", Scale="1,1,1") {
    rotationMode=BMM_Additive
    rotationSpace=BCS_BoneSpace
    translationMode=BMM_Additive
    translationSpace=BCS_BoneSpace
    boneToModify {boneName=base_skirt_bk_l}
    rotation=<rot 12 -10 0.015>
    translation {x=-4}
   }
   n35 = "Transform (Modify) Bone - Bone: base_skirt_bk_c"(ComponentPose=@n34, Alpha=1, Translation="-4,0,0", Rotation="12,0,0", Scale="1,1,1") {
    rotationMode=BMM_Additive
    rotationSpace=BCS_BoneSpace
    translationMode=BMM_Additive
    translationSpace=BCS_BoneSpace
    boneToModify {boneName=base_skirt_bk_c}
    rotation {pitch=12}
    translation {x=-4}
   }
   n36 = "Transform (Modify) Bone - Bone: base_skirt_bk_r"(ComponentPose=@n35, Alpha=1, Translation="-4,0,0", Rotation="12,0,0", Scale="1,1,1") {
    rotationMode=BMM_Additive
    rotationSpace=BCS_BoneSpace
    translationMode=BMM_Additive
    translationSpace=BCS_BoneSpace
    boneToModify {boneName=base_skirt_bk_r}
    rotation {pitch=12}
    translation {x=-4}
   }
   n37 = "Transform (Modify) Bone - Bone: dyn_skirt_bk_03_l"(ComponentPose=@n36, Alpha=1, Translation="-2,0,0", Rotation="12,0,0.01", Scale="1,1,1") {
    rotationMode=BMM_Additive
    rotationSpace=BCS_BoneSpace
    translationMode=BMM_Additive
    translationSpace=BCS_BoneSpace
    boneToModify {boneName=dyn_skirt_bk_03_l}
    rotation {pitch=12, roll=0.01}
    translation {x=-2}
   }
   n38 = "Transform (Modify) Bone - Bone: dyn_skirt_bk_03_c"(ComponentPose=@n37, Alpha=1, Translation="-2,0,0", Rotation="12,0,0", Scale="1,1,1") {
    rotationMode=BMM_Additive
    rotationSpace=BCS_BoneSpace
    translationMode=BMM_Additive
    translationSpace=BCS_BoneSpace
    boneToModify {boneName=dyn_skirt_bk_03_c}
    rotation {pitch=12}
    translation {x=-2}
   }
   n39 = "Transform (Modify) Bone - Bone: dyn_skirt_bk_03_r"(ComponentPose=@n38, Alpha=1, Translation="-2,0,0", Rotation="12,0,0", Scale="1,1,1") {
    rotationMode=BMM_Additive
    rotationSpace=BCS_BoneSpace
    translationMode=BMM_Additive
    translationSpace=BCS_BoneSpace
    boneToModify {boneName=dyn_skirt_bk_03_r}
    rotation {pitch=12}
    translation {x=-2}
   }
   n40 = "Transform (Modify) Bone - Bone: dyn_skirt_bk_04_l"(ComponentPose=@n39, Alpha=1, Translation="-2,0,0", Rotation="5,0,0.01", Scale="1,1,1") {
    rotationMode=BMM_Additive
    rotationSpace=BCS_BoneSpace
    translationMode=BMM_Additive
    translationSpace=BCS_BoneSpace
    boneToModify {boneName=dyn_skirt_bk_04_l}
    rotation {pitch=5, roll=0.01}
    translation {x=-2}
   }
   n41 = "Transform (Modify) Bone - Bone: dyn_skirt_bk_04_c"(ComponentPose=@n40, Alpha=1, Translation="-2,0,0", Rotation="5,0,0", Scale="1,1,1") {
    rotationMode=BMM_Additive
    rotationSpace=BCS_BoneSpace
    translationMode=BMM_Additive
    translationSpace=BCS_BoneSpace
    boneToModify {boneName=dyn_skirt_bk_04_c}
    rotation {pitch=5}
    translation {x=-2}
   }
   n42 = "Transform (Modify) Bone - Bone: dyn_skirt_bk_04_r"(ComponentPose=@n41, Alpha=1, Translation="-2,0,0", Rotation="5,0,0", Scale="1,1,1") {
    rotationMode=BMM_Additive
    rotationSpace=BCS_BoneSpace
    translationMode=BMM_Additive
    translationSpace=BCS_BoneSpace
    boneToModify {boneName=dyn_skirt_bk_04_r}
    rotation {pitch=5}
    translation {x=-2}
   }
   n43 = "Transform (Modify) Bone - Bone: base_skirt_fr_c"(ComponentPose=@n42, bAlphaBoolEnabled=True, Translation="0,0,0", Rotation="0,25,0", Scale="1,1,1") {
    alphaInputType=Bool
    rotationMode=BMM_Additive
    rotationSpace=BCS_BoneSpace
    translationMode=BMM_Additive
    translationSpace=BCS_BoneSpace
    boneToModify {boneName=base_skirt_fr_c}
    rotation {yaw=25}
   }
   n44 = "Transform (Modify) Bone - Bone: dyn_skirt_fr_03_c"!(ComponentPose=@n43, bAlphaBoolEnabled:bool=True, Translation="0,0,0", Rotation="-35,0,0", Scale="1,1,1") -> (Pose) {
    alphaInputType=Bool
    rotationMode=BMM_Additive
    rotationSpace=BCS_BoneSpace
    translationMode=BMM_Additive
    translationSpace=BCS_BoneSpace
    boneToModify {boneName=dyn_skirt_fr_03_c}
    rotation {pitch=-35}
   }
   n45 = "Transform (Modify) Bone - Bone: dyn_skirt_fr_04_c"!(ComponentPose=@n44, bAlphaBoolEnabled:bool=True, Translation="0,0,0", Rotation="-15,0,0", Scale="1,1,1") -> (Pose) {
    alphaInputType=Bool
    rotationMode=BMM_Additive
    rotationSpace=BCS_BoneSpace
    translationMode=BMM_Additive
    translationSpace=BCS_BoneSpace
    boneToModify {boneName=dyn_skirt_fr_04_c}
    rotation {pitch=-15}
   }
   Save cached pose 'PostAnimDynamics'(Pose=Component To Local(ComponentPose=@n45)) {cacheName=PostAnimDynamics}
   n46 = Anim Dynamics - Bone: deform_pec_l!(ComponentPose=Local To Component(LocalPose=Use cached pose 'PostAnimDynamics'()), ExternalForce="0,0,0") -> (Pose) {
    alpha=0.8
    angularSpringConstant=15000
    bLinearSpring=true
    bOverrideLinearDamping=true
    bUseGravityOverride=true
    bUsePlanarLimit=false
    linearDampingOverride=0.9995
    linearSpringConstant=400
    simulationSpace=RootRelative
    boundBone {boneName=deform_pec_l}
    componentAppliedLinearAccClamp=<1000 1000 1000>
    gravityOverride {z=-200}
    physicsBodyDefinitions [{
      boundBone {boneName=deform_pec_l}
      boxExtents=<50 55 60>
      constraintSetup {angularLimitsMax=<0 0 0>, angularLimitsMin=<0 0 0>, angularTarget=<1 0 0>, linearAxesMax=<1 0 3>, linearAxesMin=<-1 0 -3>}
      localJointOffset=<-4 0 0>
     }]
    relativeSpaceBone {boneName=pelvis}
    simSpaceSettings {maxAngularAcceleration=10, maxAngularVelocity=770, simSpaceAngularAlpha=0.4}
   }
   n47 = Anim Dynamics - Bone: deform_pec_r!(ComponentPose=@n46, ExternalForce="0,0,0") -> (Pose) {
    alpha=0.8
    angularSpringConstant=15000
    bLinearSpring=true
    bOverrideLinearDamping=true
    bUseGravityOverride=true
    bUsePlanarLimit=false
    linearDampingOverride=0.9995
    linearSpringConstant=400
    simulationSpace=RootRelative
    boundBone {boneName=deform_pec_r}
    componentAppliedLinearAccClamp=<1000 1000 1000>
    gravityOverride {z=-200}
    physicsBodyDefinitions [{
      boundBone {boneName=deform_pec_r}
      boxExtents=<50 55 60>
      constraintSetup {angularLimitsMax=<0 0 0>, angularLimitsMin=<0 0 0>, angularTarget=<1 0 0>, linearAxesMax=<1 0 3>, linearAxesMin=<-1 0 -3>}
      localJointOffset=<4 0 0>
     }]
    relativeSpaceBone {boneName=pelvis}
    simSpaceSettings {maxAngularAcceleration=10, maxAngularVelocity=770, simSpaceAngularAlpha=0.4}
   }
   Save cached pose 'PostBodyDynamics'(Pose=Component To Local(ComponentPose=@n47)) {cacheName=PostBodyDynamics}
   n48 = RigidBody!(
    ComponentPose=Local To Component(LocalPose=Use cached pose 'ControlRig Cache'())
    Alpha
    OverrideWorldGravity:Vector="0,0,0"
    ExternalForce
    SimSpaceSettings:Sim Space Settings Structure=(WorldAlpha=0.5,VelocityScaleZ=0.4,DampingAlpha=1,MaxLinearVelocity=1000,MaxAngularVelocity=10,MaxLinearAcceleration=2600,MaxAngularAcceleration=30,ExternalLinearDragV=(X=1.5,Y=1.5,Z=1.5),ExternalLinearVelocity=(X=0,Y=0,Z=0),ExternalAngularVelocity=(X=0,Y=0,Z=0))
   ) -> (Pose) {
    bOverrideWorldGravity=true
    overridePhysicsAsset=&/F_MED_GrowMallet_LeftSkirt_Physics
    simulationSpace=BaseBoneSpace
    baseBoneRef {boneName=thigh_l}
    simSpaceSettings {
     maxAngularAcceleration=30
     maxAngularVelocity=10
     maxLinearAcceleration=2600
     maxLinearVelocity=1000
     velocityScaleZ=0.4
     worldAlpha=0.5
     externalLinearDragV=<1.5 1.5 1.5>
    }
   }
   n49 = RigidBody(
    ComponentPose=@n48
    OverrideWorldGravity="0,0,0"
    SimSpaceSettings=(WorldAlpha=0.5,VelocityScaleZ=0.4,DampingAlpha=1,MaxLinearVelocity=1000,MaxAngularVelocity=10,MaxLinearAcceleration=2600,MaxAngularAcceleration=30,ExternalLinearDragV=(X=1.5,Y=1.5,Z=1.5),ExternalLinearVelocity=(X=0,Y=0,Z=0),ExternalAngularVelocity=(X=0,Y=0,Z=0))
   ) {
    bOverrideWorldGravity=true
    overridePhysicsAsset=&/F_MED_GrowMallet_RightSkirt_Physics
    simulationSpace=BaseBoneSpace
    baseBoneRef {boneName=thigh_r}
    simSpaceSettings {
     maxAngularAcceleration=30
     maxAngularVelocity=10
     maxLinearAcceleration=2600
     maxLinearVelocity=1000
     velocityScaleZ=0.4
     worldAlpha=0.5
     externalLinearDragV=<1.5 1.5 1.5>
    }
   }
   n50 = Component To Local(ComponentPose=@n49)
   n51 = RigidBody(
    ComponentPose=Local To Component(LocalPose=Use cached pose 'ControlRig Cache'())
    OverrideWorldGravity="0,0,0"
    SimSpaceSettings=(WorldAlpha=0.5,VelocityScaleZ=0.4,DampingAlpha=1,MaxLinearVelocity=1000,MaxAngularVelocity=10,MaxLinearAcceleration=2600,MaxAngularAcceleration=30,ExternalLinearDragV=(X=1.5,Y=1.5,Z=1.5),ExternalLinearVelocity=(X=0,Y=0,Z=0),ExternalAngularVelocity=(X=0,Y=0,Z=0))
   ) {
    bOverrideWorldGravity=true
    overridePhysicsAsset=&/F_MED_GrowMallet_SkirtSkydiving_Physics
    simulationSpace=BaseBoneSpace
    baseBoneRef {boneName=pelvis}
    simSpaceSettings {
     maxAngularAcceleration=30
     maxAngularVelocity=10
     maxLinearAcceleration=2600
     maxLinearVelocity=1000
     velocityScaleZ=0.4
     worldAlpha=0.5
     externalLinearDragV=<1.5 1.5 1.5>
    }
   }
   n52 = Component To Local(ComponentPose=@n51)
   n53 = Get bIsSkyDiving()
   n54 = Anim Dynamics - Bone: dyn_skirt_bk_01_c!(
    ComponentPose=Local To Component(LocalPose=Blend Poses by bool(BlendPose_0=@n52, BlendPose_1=@n50, bActiveValue=@n53))
    Alpha=1
    GravityOverride:Vector="0,0,-980"
    ExternalForce="0,0,0"
   ) -> (Pose) {
    angularDampingOverride=1
    angularSpringConstant=20000
    bAngularSpring=true
    bOverrideAngularDamping=true
    bUseGravityOverride=true
    gravityScale=0.35
    simulationSpace=RootRelative
    boundBone {boneName=dyn_skirt_bk_01_c}
    componentAppliedLinearAccClamp=<4000 4000 4000>
    componentLinearAccScale=<0.3 0.3 0.3>
    componentLinearVelScale=<0.3 0.3 0.3>
    gravityOverride {z=-980}
    physicsBodyDefinitions [{
      boundBone {boneName=dyn_skirt_bk_01_c}
      boxExtents=<20 30 30>
      constraintSetup {angularLimitsMax=<0 4 1>, angularLimitsMin=<0 -4 -1>, angularTarget=<1 0 0>}
      localJointOffset=<4 0 0>
     }]
   }
   n55 = Anim Dynamics - Bone: dyn_skirt_bk_02_c!(ComponentPose=@n54, Alpha=1, GravityOverride:Vector="0,0,-980", ExternalForce="0,0,0") -> (Pose) {
    angularDampingOverride=1
    angularSpringConstant=20000
    bAngularSpring=true
    bOverrideAngularDamping=true
    bUseGravityOverride=true
    gravityScale=0.35
    simulationSpace=RootRelative
    boundBone {boneName=dyn_skirt_bk_02_c}
    componentAppliedLinearAccClamp=<4000 4000 4000>
    componentLinearAccScale=<0.3 0.3 0.3>
    componentLinearVelScale=<0.3 0.3 0.3>
    gravityOverride {z=-980}
    physicsBodyDefinitions [{
      boundBone {boneName=dyn_skirt_bk_02_c}
      boxExtents=<20 30 30>
      constraintSetup {angularLimitsMax=<0 4 1>, angularLimitsMin=<0 -4 -1>, angularTarget=<1 0 0>}
      localJointOffset=<4 0 0>
     }]
   }
   n56 = Anim Dynamics - Bone: dyn_skirt_bk_03_c!(ComponentPose=@n55, Alpha=1, GravityOverride:Vector="0,0,-980", ExternalForce="0,0,0") -> (Pose) {
    angularDampingOverride=1
    angularSpringConstant=20000
    bAngularSpring=true
    bOverrideAngularDamping=true
    bUseGravityOverride=true
    gravityScale=0.35
    simulationSpace=RootRelative
    boundBone {boneName=dyn_skirt_bk_03_c}
    componentAppliedLinearAccClamp=<4000 4000 4000>
    componentLinearAccScale=<0.3 0.3 0.3>
    componentLinearVelScale=<0.3 0.3 0.3>
    gravityOverride {z=-980}
    physicsBodyDefinitions [{
      boundBone {boneName=dyn_skirt_bk_03_c}
      boxExtents=<20 30 30>
      constraintSetup {angularLimitsMax=<0 4 1>, angularLimitsMin=<0 -4 -1>, angularTarget=<1 0 0>}
      localJointOffset=<4 0 0>
     }]
   }
   n57 = Anim Dynamics - Bone: dyn_skirt_bk_04_c(ComponentPose=@n56, Alpha=1, GravityOverride="0,0,-980", ExternalForce="0,0,0") {
    angularDampingOverride=1
    angularSpringConstant=20000
    bAngularSpring=true
    bOverrideAngularDamping=true
    bUseGravityOverride=true
    gravityScale=0.35
    simulationSpace=RootRelative
    boundBone {boneName=dyn_skirt_bk_04_c}
    componentAppliedLinearAccClamp=<4000 4000 4000>
    componentLinearAccScale=<0.3 0.3 0.3>
    componentLinearVelScale=<0.3 0.3 0.3>
    gravityOverride {z=-980}
    physicsBodyDefinitions [{
      boundBone {boneName=dyn_skirt_bk_04_c}
      boxExtents=<20 30 30>
      constraintSetup {angularLimitsMax=<0 4 1>, angularLimitsMin=<0 -4 -1>, angularTarget=<1 0 0>}
      localJointOffset=<4 0 0>
     }]
   }
   n58 = Anim Dynamics - Bone: dyn_skirt_bk_01_r!(ComponentPose=@n57, Alpha=1, GravityOverride:Vector="0,0,-980", ExternalForce="0,0,0") -> (Pose) {
    angularDampingOverride=1
    angularSpringConstant=15000
    bAngularSpring=true
    bOverrideAngularDamping=true
    bUseGravityOverride=true
    gravityScale=0.35
    simulationSpace=RootRelative
    boundBone {boneName=dyn_skirt_bk_01_r}
    componentAppliedLinearAccClamp=<4000 4000 4000>
    componentLinearAccScale=<0.3 0.3 0.35>
    componentLinearVelScale=<0.3 0.3 0.35>
    gravityOverride {z=-980}
    physicsBodyDefinitions [{
      boundBone {boneName=dyn_skirt_bk_01_r}
      boxExtents=<25 35 40>
      constraintSetup {angularLimitsMax=<0 4 1>, angularLimitsMin=<0 -4 -1>, angularTarget=<1 0 0>}
      localJointOffset=<4 0 0>
     }]
    simSpaceSettings {maxAngularAcceleration=100, maxAngularVelocity=10, simSpaceAngularAlpha=1}
   }
   n59 = Anim Dynamics - Bone: dyn_skirt_bk_02_r!(ComponentPose=@n58, Alpha=1, GravityOverride:Vector="0,0,-980", ExternalForce="0,0,0") -> (Pose) {
    angularDampingOverride=1
    angularSpringConstant=15000
    bAngularSpring=true
    bOverrideAngularDamping=true
    bUseGravityOverride=true
    gravityScale=0.35
    simulationSpace=RootRelative
    boundBone {boneName=dyn_skirt_bk_02_r}
    componentAppliedLinearAccClamp=<4000 4000 4000>
    componentLinearAccScale=<0.3 0.3 0.35>
    componentLinearVelScale=<0.3 0.3 0.35>
    gravityOverride {z=-980}
    physicsBodyDefinitions [{
      boundBone {boneName=dyn_skirt_bk_02_r}
      boxExtents=<25 35 40>
      constraintSetup {angularLimitsMax=<0 4 1>, angularLimitsMin=<0 -4 -1>, angularTarget=<1 0 0>}
      localJointOffset=<4 0 0>
     }]
    simSpaceSettings {maxAngularAcceleration=100, maxAngularVelocity=10, simSpaceAngularAlpha=1}
   }
   n60 = Anim Dynamics - Bone: dyn_skirt_bk_03_r!(ComponentPose=@n59, Alpha=1, GravityOverride:Vector="0,0,-980", ExternalForce="0,0,0") -> (Pose) {
    angularDampingOverride=1
    angularSpringConstant=15000
    bAngularSpring=true
    bOverrideAngularDamping=true
    bUseGravityOverride=true
    gravityScale=0.35
    simulationSpace=RootRelative
    boundBone {boneName=dyn_skirt_bk_03_r}
    componentAppliedLinearAccClamp=<4000 4000 4000>
    componentLinearAccScale=<0.3 0.3 0.35>
    componentLinearVelScale=<0.3 0.3 0.35>
    gravityOverride {z=-980}
    physicsBodyDefinitions [{
      boundBone {boneName=dyn_skirt_bk_03_r}
      boxExtents=<25 35 40>
      constraintSetup {angularLimitsMax=<0 4 1>, angularLimitsMin=<0 -4 -1>, angularTarget=<1 0 0>}
      localJointOffset=<4 0 0>
     }]
    simSpaceSettings {maxAngularAcceleration=100, maxAngularVelocity=10, simSpaceAngularAlpha=1}
   }
   n61 = Anim Dynamics - Bone: dyn_skirt_bk_04_r(ComponentPose=@n60, Alpha=1, GravityOverride="0,0,-980", ExternalForce="0,0,0") {
    angularDampingOverride=1
    angularSpringConstant=15000
    bAngularSpring=true
    bOverrideAngularDamping=true
    bUseGravityOverride=true
    gravityScale=0.35
    simulationSpace=RootRelative
    boundBone {boneName=dyn_skirt_bk_04_r}
    componentAppliedLinearAccClamp=<4000 4000 4000>
    componentLinearAccScale=<0.3 0.3 0.35>
    componentLinearVelScale=<0.3 0.3 0.35>
    gravityOverride {z=-980}
    physicsBodyDefinitions [{
      boundBone {boneName=dyn_skirt_bk_04_r}
      boxExtents=<25 35 40>
      constraintSetup {angularLimitsMax=<0 4 1>, angularLimitsMin=<0 -4 -1>, angularTarget=<1 0 0>}
      localJointOffset=<4 0 0>
     }]
    simSpaceSettings {maxAngularAcceleration=100, maxAngularVelocity=10, simSpaceAngularAlpha=1}
   }
   n62 = Anim Dynamics - Bone: dyn_skirt_bk_01_l!(ComponentPose=@n61, Alpha=1, GravityOverride:Vector="0,0,-980", ExternalForce="0,0,0") -> (Pose) {
    angularDampingOverride=1
    angularSpringConstant=15000
    bAngularSpring=true
    bOverrideAngularDamping=true
    bUseGravityOverride=true
    gravityScale=0.35
    simulationSpace=RootRelative
    boundBone {boneName=dyn_skirt_bk_01_l}
    componentAppliedLinearAccClamp=<4000 4000 4000>
    componentLinearAccScale=<0.3 0.3 0.35>
    componentLinearVelScale=<0.3 0.3 0.35>
    gravityOverride {z=-980}
    physicsBodyDefinitions [{
      boundBone {boneName=dyn_skirt_bk_01_l}
      boxExtents=<25 35 40>
      constraintSetup {angularLimitsMax=<0 4 1>, angularLimitsMin=<0 -4 -1>, angularTarget=<1 0 0>}
      localJointOffset=<4 0 0>
     }]
    simSpaceSettings {maxAngularAcceleration=100, maxAngularVelocity=10, simSpaceAngularAlpha=1}
   }
   n63 = Anim Dynamics - Bone: dyn_skirt_bk_02_l!(ComponentPose=@n62, Alpha=1, GravityOverride:Vector="0,0,-980", ExternalForce="0,0,0") -> (Pose) {
    angularDampingOverride=1
    angularSpringConstant=15000
    bAngularSpring=true
    bOverrideAngularDamping=true
    bUseGravityOverride=true
    gravityScale=0.35
    simulationSpace=RootRelative
    boundBone {boneName=dyn_skirt_bk_02_l}
    componentAppliedLinearAccClamp=<4000 4000 4000>
    componentLinearAccScale=<0.3 0.3 0.35>
    componentLinearVelScale=<0.3 0.3 0.35>
    gravityOverride {z=-980}
    physicsBodyDefinitions [{
      boundBone {boneName=dyn_skirt_bk_02_l}
      boxExtents=<25 35 40>
      constraintSetup {angularLimitsMax=<0 4 1>, angularLimitsMin=<0 -4 -1>, angularTarget=<1 0 0>}
      localJointOffset=<4 0 0>
     }]
    simSpaceSettings {maxAngularAcceleration=100, maxAngularVelocity=10, simSpaceAngularAlpha=1}
   }
   n64 = Anim Dynamics - Bone: dyn_skirt_bk_03_l!(ComponentPose=@n63, Alpha=1, GravityOverride:Vector="0,0,-980", ExternalForce="0,0,0") -> (Pose) {
    angularDampingOverride=1
    angularSpringConstant=15000
    bAngularSpring=true
    bOverrideAngularDamping=true
    bUseGravityOverride=true
    gravityScale=0.35
    simulationSpace=RootRelative
    boundBone {boneName=dyn_skirt_bk_03_l}
    componentAppliedLinearAccClamp=<4000 4000 4000>
    componentLinearAccScale=<0.3 0.3 0.35>
    componentLinearVelScale=<0.3 0.3 0.35>
    gravityOverride {z=-980}
    physicsBodyDefinitions [{
      boundBone {boneName=dyn_skirt_bk_03_l}
      boxExtents=<25 35 40>
      constraintSetup {angularLimitsMax=<0 4 1>, angularLimitsMin=<0 -4 -1>, angularTarget=<1 0 0>}
      localJointOffset=<4 0 0>
     }]
    simSpaceSettings {maxAngularAcceleration=100, maxAngularVelocity=10, simSpaceAngularAlpha=1}
   }
   n65 = Anim Dynamics - Bone: dyn_skirt_bk_04_l(ComponentPose=@n64, Alpha=1, GravityOverride="0,0,-980", ExternalForce="0,0,0") {
    angularDampingOverride=1
    angularSpringConstant=15000
    bAngularSpring=true
    bOverrideAngularDamping=true
    bUseGravityOverride=true
    gravityScale=0.35
    simulationSpace=RootRelative
    boundBone {boneName=dyn_skirt_bk_04_l}
    componentAppliedLinearAccClamp=<4000 4000 4000>
    componentLinearAccScale=<0.3 0.3 0.35>
    componentLinearVelScale=<0.3 0.3 0.35>
    gravityOverride {z=-980}
    physicsBodyDefinitions [{
      boundBone {boneName=dyn_skirt_bk_04_l}
      boxExtents=<25 35 40>
      constraintSetup {angularLimitsMax=<0 4 1>, angularLimitsMin=<0 -4 -1>, angularTarget=<1 0 0>}
      localJointOffset=<4 0 0>
     }]
    simSpaceSettings {maxAngularAcceleration=100, maxAngularVelocity=10, simSpaceAngularAlpha=1}
   }
   Save cached pose 'Initial Cache Pose'(Pose=Component To Local(ComponentPose=@n65)) {cacheName=Initial Cache Pose}
  ]}
 EventGraph {nodes [
   Event BlueprintUpdateAnimation()
   > n1 = Sequence!(execute) -> (then_0, then_1, then_2:Exec, then_3:Exec, then_4:Exec)
   > n2 = "Gravity Override State Function (Parameter Asset) (Nativized)"(
    CharacterPartAnimInstance=Self-Reference()
    GameTimeInSeconds=GetGameTimeInSeconds()
    DebugJointName=spine_05
    Parameters=&/F_MED_GrowMallet_Parameters
   )
   > Set Gravity Override(Gravity Override=@n2)
   > Set RBAN Settings Out(RBAN Settings Out=@n2.OutRigidBodyAnimNodeParameters)
   n3 = Sequence!(execute=@n1.then_4) -> (then_0, then_1, then_2, then_3, then_4, then_5:Exec, then_6:Exec)
   n4 = Get bIsCrouching()
   n5 = Get bIsSkyDiving()
   n6 = Get bIsMoving2D()
   n7 = Get bIsTacticalSprinting()
   > n8 = Branch(Condition=AND Boolean!(A=OR Boolean(A=@n6, B=@n7), B=NOT Boolean(A=@n4), C:bool=NOT Boolean(A=@n5)) -> (ReturnValue))
   > Set Sprinting_Running(Sprinting_Running:double=1) -> (Output_Get:double)
   Set Sprinting_Running(execute=@n8.else, Sprinting_Running=0)
   n9 = Get bIsCrouching()
   n10 = Branch(execute=@n3.then_1, Condition=AND Boolean!(A=Get bIsMoving2D(), B=NOT Boolean(A=@n9)) -> (ReturnValue))
   > Set Moving2D(Moving2D:double=1) -> (Output_Get:double)
   Set Moving2D(execute=@n10.else, Moving2D=0)
   n11 = Get CurrentLOD()
   n12 = Branch(execute=@n3.then_4, Condition=OR Boolean(A="integer > integer"(A=@n11, B=1) {is=title K2Node_CallFunction}, B=Get bIsDBNOCarried()))
   > Set LODsCheck(LODsCheck:double=1) -> (Output_Get:double)
   > SetMorphTarget!(execute, self:Anim Instance Object Reference, MorphTargetName:Name=SkirtBody_LOD_BS, Value:float=1) -> (then)
   Set LODsCheck(execute=@n12.else, LODsCheck=0)
   > SetMorphTarget(MorphTargetName=SkirtBody_LOD_BS, Value=0)
   n13 = GetPartSkeletalMeshComponent()
   Set isASW(
    execute=@n3.then_5
    isASW:bool="Equal (Name)"(
     A=GetCurrentSkinWeightProfileName!(self:Skinned Mesh Component Object Reference=@n13, InLayer:ESkinWeightProfileLayer Enum=Primary) -> (ReturnValue:Name=None)
     B=Alternative
    )
   ) -> (Output_Get:bool)
   n14 = Branch(execute=@n3.then_6, Condition="integer > integer"(A=Get CurrentLOD(), B=0))
   > SetMorphTarget(MorphTargetName=SkirtBody_LOD_BS, Value=2)
   SetMorphTarget(execute=@n14.else, MorphTargetName=SkirtBody_LOD_BS, Value=0)
   n15 = Branch(execute=@n3.then_2, Condition=Get bIsOnGrindRail())
   > Set isGrindRail(isGrindRail:double=1) -> (Output_Get:double)
   Set isGrindRail(execute=@n15.else, isGrindRail=0)
   n16 = Get bIsMoving2D()
   n17 = Branch(execute=@n3.then_3, Condition=AND Boolean(A=Get bIsCrouching(), B=NOT Boolean(A=@n16)))
   n18 = GetCurveValue(CurveName=CrouchingCurve)
   > Set SkirtPostAlpha(SkirtPostAlpha:double=@n18) -> (Output_Get:double)
   Set SkirtPostAlpha(execute=@n17.else, SkirtPostAlpha=@n18)
   n19 = OR Boolean!(A=Get bIsKneeSliding(), B=Get bIsCrouchSprinting(), C:bool=Get bIsMoving2D()) -> (ReturnValue)
   n20 = Branch(execute=@n1.then_2, Condition=AND Boolean(A=Get bIsCrouching(), B=NOT Boolean(A=@n19)))
   > Set Crouching(Crouching:double=1) -> (Output_Get:double)
   Set Crouching(execute=@n20.else, Crouching=0)
   n21 = PlayerMovement(execute=@n1.then_3, AnimInstance=Self-Reference(), XDotProductCoefficient=1, YDotProductCoefficient=1, ZDotProductCoefficient=1)
   n22 = BreakVector(InVec=MakeVector(X=@n21, Y=@n21.YMovement))
   > Set strafeLeft float(strafeLeft float:double="Absolute (Float)"(A="Clamp (Float)"(Value=@n22, Min=-1, Max=0))) -> (Output_Get:double)
   > Set strafeRight float(strafeRight float:double="Clamp (Float)"(Value=@n22, Min=0, Max=1)) -> (Output_Get:double)
   > Set moveForward(moveForward:double="Clamp (Float)"(Value=@n22.Y, Min=0, Max=1)) -> (Output_Get:double)
   > Set moveBackward(moveBackward:double="Absolute (Float)"(A="Clamp (Float)"(Value=@n22.Y, Min=-1, Max=0))) -> (Output_Get:double)
   > n23 = Branch(Condition=Get bIsPlayingMeleeAnim())
   > Set Alpha_SkirtDynamics(Alpha_SkirtDynamics:double=0.38) -> (Output_Get:double)
   n24 = Branch(execute=@n23.else, Condition=Get bIsSkyDiving())
   > Set Alpha_SkirtDynamics(Alpha_SkirtDynamics=0.5)
   n25 = Branch(execute=@n24.else, Condition=Get bIsPlayingEmote())
   > Set Alpha_SkirtDynamics(Alpha_SkirtDynamics=0.5)
   n26 = Branch(execute=@n25.else, Condition=Get bIsCrouching())
   > Set Alpha_SkirtDynamics(Alpha_SkirtDynamics=0.5)
   n27 = Branch(execute=@n26.else, Condition=Get bIsSkydiveFloating())
   > Set Alpha_SkirtDynamics(Alpha_SkirtDynamics=0.45)
   n28 = Branch(execute=@n27.else, Condition=Get bIsJumping())
   > Set Alpha_SkirtDynamics(Alpha_SkirtDynamics=0.85)
   n29 = Branch(execute=@n28.else, Condition=Get bIsSprinting())
   > Set Alpha_SkirtDynamics(Alpha_SkirtDynamics=0.75)
   n30 = Branch(execute=@n29.else, Condition=Get bIsInFrontEnd())
   > Set Alpha_SkirtDynamics(Alpha_SkirtDynamics=0.4)
   n31 = Branch(execute=@n30.else, Condition=Get bIsDBNOCarried())
   > Set Alpha_SkirtDynamics(Alpha_SkirtDynamics=0)
   Set Alpha_SkirtDynamics(execute=@n31.else, Alpha_SkirtDynamics=0.9)
   n32 = Sequence!(execute=@n1.then_1) -> (then_0, then_1, then_2, then_3, then_4)
   n33 = Branch(execute=@n32.then_1, Condition=OR Boolean!(A=Get bIsCrouching(), B=Get bIsCrouchSprinting()) -> (ReturnValue))
   > Set RegularCrouching(RegularCrouching:double=1) -> (Output_Get:double)
   Set RegularCrouching(execute=@n33.else, RegularCrouching=0)
   n34 = Branch(execute=@n32.then_2, Condition=Get bIsSkydiveFloating())
   > Set SkyDiving_Alpha(SkyDiving_Alpha:double=1) -> (Output_Get:double)
   Set SkyDiving_Alpha(execute=@n34.else, SkyDiving_Alpha=0)
   Set Is_Parachute_Open(execute=@n32.then_3, Is_Parachute_Open:double="To Float (Boolean)"(InBool=Get bIsParachuteOpened())) -> (Output_Get:double)
   n35 = Branch(execute=@n32.then_4, Condition=Get bIsJumping())
   > Set isJumping(isJumping:double=1) -> (Output_Get:double)
   Delay(execute=@n35.else, Duration=0.2)
   > Set isJumping(isJumping=0)
   n36 = Branch(execute=@n32.then_0, Condition=AND Boolean(A=Get bIsCrouching(), B=Get bIsMoving2D()))
   > Set Crouching_sideWays(Crouching_sideWays:double=1) -> (Output_Get:double)
   Set Crouching_sideWays(execute=@n36.else, Crouching_sideWays=0)
   Set Alpha_SkirtDynamics(Alpha_SkirtDynamics=0.9)
  ]}
]
variables [
 Alpha_SkirtDynamics {type=number}
 Crouching {type=number}
 Crouching_sideWays {type=number}
 Gravity Override {type=Vector}
 Is_on_Air {type=number}
 Is_Parachute_Open {type=number}
 isASW {type=boolean}
 isGrindRail {type=number}
 isJumping {type=number}
 LODsCheck {type=number}
 moveBackward {type=number}
 moveForward {type=number}
 Moving2D {type=number}
 RBAN Settings Out {type=FortRigidBodyAnimNodeParameters}
 RegularCrouching {type=number}
 Skeletal Mesh {type=SkeletalMeshComponent}
 SkirtPostAlpha {type=number}
 SkyDiving_Alpha {type=number}
 Sprinting_Running {type=number}
 strafeLeft float {type=number}
 strafeRight float {type=number}
]
