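# Physics asset for the skirt of the F_MED_GrowMallet female player body, named as a backup of its colliders.
# Four skirt chains (front and side, left and right) each link four capsule bodies under a small base capsule constrained to the pelvis.
# Base joints swing and twist freely, while chain links carry tight swing and twist limit angles.
# Pelvis and lower spine have large capsules; thigh and calf bodies have no shapes.
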PhysicsAsset /BRCosmetics/Characters/Player/Female/Medium/Bodies/F_MED_GrowMallet/Meshes/F_MED_GrowMallet_Skirt_Colliders_Backup
bodies {
 base_skirt_fr_l [{shapeType=Capsule, length=0.505, radius=0.505, shapeName=base_skirt_fr_l_capsule, rotation=<rot 0 -90 0>}]
 base_skirt_fr_r [{shapeType=Capsule, length=0.505, radius=0.505, shapeName=base_skirt_fr_r_capsule, rotation=<rot 0 -90 0>}]
 base_skirt_sd_l [{shapeType=Capsule, length=0.505, radius=0.505, shapeName=base_skirt_sd_l_capsule, rotation=<rot 0 -90 0>}]
 base_skirt_sd_r [{shapeType=Capsule, length=0.505, radius=0.505, shapeName=base_skirt_sd_r_capsule, rotation=<rot 0 -90 0>}]
 calf_l []
 calf_r []
 dyn_skirt_fr_01_l [{shapeType=Capsule, length=6, radius=4, shapeName=dyn_skirt_fr_01_l_capsule, center=<6 0 5>, rotation=<rot 90 90 0>}]
 dyn_skirt_fr_01_r [{shapeType=Capsule, length=6, radius=4, shapeName=dyn_skirt_fr_01_r_capsule, center=<6 0 5>, rotation=<rot 90 90 0>}]
 dyn_skirt_fr_02_l [{shapeType=Capsule, length=6, radius=4, shapeName=dyn_skirt_fr_02_l_capsule, center=<6 0 5>, rotation=<rot 90 90 0>}]
 dyn_skirt_fr_02_r [{shapeType=Capsule, length=6, radius=4, shapeName=dyn_skirt_fr_02_r_capsule, center=<6 0 5>, rotation=<rot 90 90 0>}]
 dyn_skirt_fr_03_l [{shapeType=Capsule, length=6, radius=4, shapeName=dyn_skirt_fr_03_l_capsule, center=<6 0 5>, rotation=<rot 90 90 0>}]
 dyn_skirt_fr_03_r [{shapeType=Capsule, length=6, radius=4, shapeName=dyn_skirt_fr_03_r_capsule, center=<6 0 5>, rotation=<rot 90 90 0>}]
 dyn_skirt_fr_04_l [{shapeType=Capsule, length=6, radius=4, shapeName=dyn_skirt_fr_04_l_capsule, center=<6 0 5>, rotation=<rot 90 90 0>}]
 dyn_skirt_fr_04_r [{shapeType=Capsule, length=6, radius=4, shapeName=dyn_skirt_fr_04_r_capsule, center=<6 0 5>, rotation=<rot 90 90 0>}]
 dyn_skirt_sd_01_l [{shapeType=Capsule, length=6, radius=4.5, shapeName=dyn_skirt_sd_01_l_capsule, center=<6 0 3>, rotation=<rot 90 90 0>}]
 dyn_skirt_sd_01_r [{shapeType=Capsule, length=6, radius=4.5, shapeName=dyn_skirt_sd_01_r_capsule, center=<6 0 3>, rotation=<rot 90 90 0>}]
 dyn_skirt_sd_02_l [{shapeType=Capsule, length=6, radius=4.5, shapeName=dyn_skirt_sd_02_l_capsule, center=<6 0 3>, rotation=<rot 90 90 0>}]
 dyn_skirt_sd_02_r [{shapeType=Capsule, length=6, radius=4.5, shapeName=dyn_skirt_sd_02_r_capsule, center=<6 0 3>, rotation=<rot 90 90 0>}]
 dyn_skirt_sd_03_l [{shapeType=Capsule, length=6, radius=4.5, shapeName=dyn_skirt_sd_03_l_capsule, center=<6 0 3>, rotation=<rot 90 90 0>}]
 dyn_skirt_sd_03_r [{shapeType=Capsule, length=6, radius=4.5, shapeName=dyn_skirt_sd_03_r_capsule, center=<6 0 3>, rotation=<rot 90 90 0>}]
 dyn_skirt_sd_04_l [{shapeType=Capsule, length=6, radius=4.5, shapeName=dyn_skirt_sd_04_l_capsule, center=<6 0 3>, rotation=<rot 90 90 0>}]
 dyn_skirt_sd_04_r [{shapeType=Capsule, length=6, radius=4.5, shapeName=dyn_skirt_sd_04_r_capsule, center=<6 0 3>, rotation=<rot 90 90 0>}]
 pelvis [{shapeType=Capsule, length=8.1042, radius=11.6441, center=<-2.4471 -3.2735 -0.1981>}]
 spine_01 [{shapeType=Capsule, length=2.8888, radius=11.8638, center=<1.4044 -0.4776 0>}]
 spine_02 [{shapeType=Capsule, length=3, radius=11.3638, center=<3.689 1.4151 0>}]
 thigh_l []
 thigh_r []
}
constraints [
 {
  bone1Name=base_skirt_fr_l
  bone2Name=pelvis
  swing1LimitDegrees=45
  swing1Motion=Free
  swing2LimitDegrees=45
  swing2Motion=Free
  twistLimitDegrees=45
  twistMotion=Free
 }
 {bone1Name=dyn_skirt_fr_01_l, bone2Name=base_skirt_fr_l, swing1LimitDegrees=10, swing2LimitDegrees=10, twistLimitDegrees=1}
 {bone1Name=dyn_skirt_fr_02_l, bone2Name=dyn_skirt_fr_01_l, swing1LimitDegrees=10, swing2LimitDegrees=10, twistLimitDegrees=1}
 {bone1Name=dyn_skirt_fr_03_l, bone2Name=dyn_skirt_fr_02_l, swing1LimitDegrees=10, swing2LimitDegrees=10, twistLimitDegrees=1}
 {bone1Name=dyn_skirt_fr_04_l, bone2Name=dyn_skirt_fr_03_l, swing1LimitDegrees=10, swing2LimitDegrees=10, twistLimitDegrees=1}
 {
  bone1Name=base_skirt_sd_l
  bone2Name=pelvis
  swing1LimitDegrees=45
  swing1Motion=Free
  swing2LimitDegrees=45
  swing2Motion=Free
  twistLimitDegrees=45
  twistMotion=Free
 }
 {bone1Name=dyn_skirt_sd_01_l, bone2Name=base_skirt_sd_l, swing1LimitDegrees=10, swing2LimitDegrees=10, twistLimitDegrees=1}
 {bone1Name=dyn_skirt_sd_02_l, bone2Name=dyn_skirt_sd_01_l, swing1LimitDegrees=10, swing2LimitDegrees=10, twistLimitDegrees=1}
 {bone1Name=dyn_skirt_sd_03_l, bone2Name=dyn_skirt_sd_02_l, swing1LimitDegrees=10, swing2LimitDegrees=10, twistLimitDegrees=1}
 {bone1Name=dyn_skirt_sd_04_l, bone2Name=dyn_skirt_sd_03_l, swing1LimitDegrees=10, swing2LimitDegrees=10, twistLimitDegrees=1}
 {
  bone1Name=base_skirt_fr_r
  bone2Name=pelvis
  swing1LimitDegrees=45
  swing1Motion=Free
  swing2LimitDegrees=45
  swing2Motion=Free
  twistLimitDegrees=45
  twistMotion=Free
 }
 {bone1Name=dyn_skirt_fr_01_r, bone2Name=base_skirt_fr_r, swing1LimitDegrees=10, swing2LimitDegrees=10, twistLimitDegrees=1}
 {bone1Name=dyn_skirt_fr_02_r, bone2Name=dyn_skirt_fr_01_r, swing1LimitDegrees=10, swing2LimitDegrees=10, twistLimitDegrees=1}
 {bone1Name=dyn_skirt_fr_03_r, bone2Name=dyn_skirt_fr_02_r, swing1LimitDegrees=10, swing2LimitDegrees=10, twistLimitDegrees=1}
 {bone1Name=dyn_skirt_fr_04_r, bone2Name=dyn_skirt_fr_03_r, swing1LimitDegrees=10, swing2LimitDegrees=10, twistLimitDegrees=1}
 {
  bone1Name=base_skirt_sd_r
  bone2Name=pelvis
  swing1LimitDegrees=45
  swing1Motion=Free
  swing2LimitDegrees=45
  swing2Motion=Free
  twistLimitDegrees=45
  twistMotion=Free
 }
 {bone1Name=dyn_skirt_sd_01_r, bone2Name=base_skirt_sd_r, swing1LimitDegrees=10, swing2LimitDegrees=10, twistLimitDegrees=1}
 {bone1Name=dyn_skirt_sd_02_r, bone2Name=dyn_skirt_sd_01_r, swing1LimitDegrees=10, swing2LimitDegrees=10, twistLimitDegrees=1}
 {bone1Name=dyn_skirt_sd_03_r, bone2Name=dyn_skirt_sd_02_r, swing1LimitDegrees=10, swing2LimitDegrees=10, twistLimitDegrees=1}
 {bone1Name=dyn_skirt_sd_04_r, bone2Name=dyn_skirt_sd_03_r, swing1LimitDegrees=10, swing2LimitDegrees=10, twistLimitDegrees=1}
]
