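# Physics asset for the face accessory part of the female medium GrowMallet player body, set up for dynamic hair bones.
# Centre, left, right and front-left capsule chains hang from tiny offset anchor bodies with tight swing and twist limits, plus looser side branches on left and right.
# Angularly free constraints cross-link the outer chain tips to the centre tip.
# Head, neck, clavicle, upper spine and upper-arm capsules provide body collision shapes.
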PhysicsAsset /BRCosmetics/Characters/Player/Female/Medium/Bodies/F_MED_GrowMallet/Meshes/Parts/F_MED_GrowMallet_FaceAcc_Physics
bodies {
 base_hair_02_lt_helper [{shapeType=Capsule, length=4, radius=2.5, shapeName=base_hair_02_lt_helper_capsule, center=<0 -0.5 0>}]
 base_hair_02_rt_helper [{shapeType=Capsule, length=4, radius=2, shapeName=dyn_hair_02_rt_helper_capsule, center=<0.3606 0 0.3>, rotation=<rot 56.1806 180 180>}]
 clavicle_l [{
   shapeType=Capsule
   length=8.1483
   radius=5.0631
   shapeName=clavicle_l_capsule
   center=<6.6566 -1.7068 0.5882>
   rotation=<rot 81.344 -0.0676 179.9308>
  }]
 clavicle_r [{
   shapeType=Capsule
   length=7.4013
   radius=5.929
   shapeName=clavicle_r_capsule
   center=<-6.2251 1.0912 2.1208>
   rotation=<rot 87.5685 -179.9942 -179.9944>
  }]
 dyn_hair_01_cr [{shapeType=Capsule, length=4, radius=4, shapeName=dyn_hair_01_rt_capsule, center=<4 0 0>, rotation=<rot 90 0 0>}]
 dyn_hair_01_lf [{shapeType=Capsule, length=4, radius=4, shapeName=dyn_hair_01_rt_capsule, center=<4 0 0>, rotation=<rot 90 0 0>}]
 dyn_hair_01_rt [{shapeType=Capsule, length=4, radius=4, shapeName=dyn_hair_01_rt_capsule, center=<4 0 0>, rotation=<rot 90 0 0>}]
 dyn_hair_03_cr [{shapeType=Capsule, length=5, radius=3, shapeName=dyn_hair_03_cr_capsule, center=<2.5 0 0>, rotation=<rot 90 0 0>}]
 dyn_hair_03_lf [{shapeType=Capsule, length=5, radius=3, shapeName=dyn_hair_03_lf_capsule, center=<2.5 0 0>, rotation=<rot 90 0 0>}]
 dyn_hair_03_rt [{shapeType=Capsule, length=5, radius=3, shapeName=dyn_hair_03_rt_capsule, center=<2.5 0 0>, rotation=<rot 90 0 0>}]
 dyn_hair_04_cr [{shapeType=Capsule, length=4, radius=3, shapeName=dyn_hair_04_cr_capsule, center=<3.5 0 0>, rotation=<rot 90 0 0>}]
 dyn_hair_04_lf [{shapeType=Capsule, length=4, radius=3, shapeName=dyn_hair_04_lf_capsule, center=<3.5 0 0>, rotation=<rot 90 0 0>}]
 dyn_hair_04_rt [{shapeType=Capsule, length=4, radius=3, shapeName=dyn_hair_04_rt_capsule, center=<3.5 0 0>, rotation=<rot 90 0 0>}]
 dyn_hair_04_sd_l [{
   shapeType=Capsule
   length=3.3702
   radius=5.1857
   shapeName=dyn_hair_04_sd_l_capsule
   center=<1.6495 -2.5799 0.0634>
   rotation=<rot -10.255 0 -0.0001>
  }]
 dyn_hair_04_sd_r [{shapeType=Capsule, length=4, radius=4, shapeName=dyn_hair_04_sd_r_capsule, center=<2 1.5 0>}]
 dyn_hair_05_cr [{shapeType=Capsule, length=7.5, radius=3, shapeName=dyn_hair_05_cr_capsule, center=<4.5 0 0>, rotation=<rot 90 0 0>}]
 dyn_hair_05_lf [{shapeType=Capsule, length=7.5, radius=3, shapeName=dyn_hair_05_lf_capsule, center=<4.5 0 0>, rotation=<rot 90 0 0>}]
 dyn_hair_05_rt [{shapeType=Capsule, length=7.5, radius=3, shapeName=dyn_hair_05_rt_capsule, center=<4.5 0 0>, rotation=<rot 90 0 0>}]
 dyn_hair_front_01_lf [{shapeType=Capsule, length=8.1845, radius=3.865, shapeName=dyn_hair_front_01_lf_capsule, center=<6 -0.5 0>, rotation=<rot 90 0 0>}]
 dyn_hair_front_02_lf [{
   shapeType=Capsule
   length=6.008
   radius=3.2666
   shapeName=dyn_hair_front_02_lf_capsule
   center=<3.0142 -1 1>
   rotation=<rot 84.8423 -180 -180>
  }]
 dyn_hair_front_03_lf [
  {
   shapeType=Capsule
   length=6.2756
   radius=3.2244
   shapeName=dyn_hair_front_03_lf_capsule
   center=<2.7928 0 1.5>
   rotation=<rot 74.0045 176.0859 166.7477>
  }
  {
   shapeType=Capsule
   length=6.5356
   radius=2.9644
   shapeName=dyn_hair_front_03_lf_capsule
   center=<1.1165 0.9452 4.205>
   rotation=<rot 75.9689 -173.0385 170.424>
  }
 ]
 head [{shapeType=Capsule, length=4.8975, radius=10.8589, shapeName=head_capsule, center=<0 0.2511 0>, rotation=<rot 90 0 0>}]
 neck_01 [{shapeType=Capsule, length=5.6236, radius=4.2828, shapeName=neck_01_capsule, center=<0.2313 0 -1.3905>, rotation=<rot 90 0 0>}]
 offset_hair_cr [{shapeType=Capsule, length=0.505, radius=0.505, shapeName=offset_hair_cr_capsule, rotation=<rot 0 -90 0>}]
 offset_hair_front_lf [{shapeType=Capsule, length=0.505, radius=0.505, shapeName=offset_hair_front_lf_capsule, rotation=<rot 0 -90 0>}]
 offset_hair_lf [{shapeType=Capsule, length=0.505, radius=0.505, shapeName=offset_hair_lf_capsule, rotation=<rot 0 -90 0>}]
 offset_hair_rt [{shapeType=Capsule, length=0.505, radius=0.505, shapeName=offset_hair_rt_capsule, rotation=<rot 0 -90 0>}]
 spine_04 [
  {shapeType=Capsule, length=16.0862, radius=8.9138, shapeName=spine_05_capsule, center=<-10.9301 -0.7195 0>, rotation=<rot 0 -90 0>}
  {shapeType=Capsule, length=17.2664, radius=8.9449, shapeName=spine_04_capsule, center=<0.8439 -1.824 0>}
  {shapeType=Capsule, length=14.5704, radius=8.0067, shapeName=spine_04_capsule, center=<0.1101 3.0149 0>}
 ]
 spine_05 [{shapeType=Capsule, length=18.7989, radius=8.11, shapeName=spine_05_capsule, center=<-1.7848 -0.9112 0>, rotation=<rot 0 -90 0>}]
 upperarm_l [
  {
   shapeType=Capsule
   length=27.5556
   radius=5.4328
   shapeName=upperarm_l_capsule
   center=<15.0744 -0.7619 -0.4596>
   rotation=<rot -0.0202 -94.0817 -90.2794>
  }
  {
   shapeType=Capsule
   length=27.5556
   radius=4.4741
   shapeName=upperarm_l_capsule
   center=<21.5128 -2.4874 9.6038>
   rotation=<rot -0.0203 -94.0817 -90.2794>
  }
 ]
 upperarm_r [{shapeType=Capsule, length=24.0453, radius=5.0231, shapeName=upperarm_l_capsule, center=<-13.9807 0.352 0.5342>, rotation=<rot 0 -90 -90>}]
}
constraints [
 {bone1Name=dyn_hair_01_cr, bone2Name=offset_hair_cr, swing1LimitDegrees=8, swing2LimitDegrees=10, twistLimitDegrees=0}
 {bone1Name=dyn_hair_01_lf, bone2Name=offset_hair_lf, swing1LimitDegrees=4, swing2LimitDegrees=10, twistLimitDegrees=0}
 {bone1Name=dyn_hair_01_rt, bone2Name=offset_hair_rt, swing1LimitDegrees=8, swing2LimitDegrees=10, twistLimitDegrees=0}
 {bone1Name=dyn_hair_front_02_lf, bone2Name=dyn_hair_front_01_lf, swing1LimitDegrees=5, swing2LimitDegrees=20, twistLimitDegrees=0}
 {bone1Name=dyn_hair_front_03_lf, bone2Name=dyn_hair_front_02_lf, swing1LimitDegrees=5, swing2LimitDegrees=20, twistLimitDegrees=0}
 {bone1Name=dyn_hair_front_01_lf, bone2Name=offset_hair_front_lf, swing1LimitDegrees=4, swing2LimitDegrees=12, twistLimitDegrees=3}
 {bone1Name=dyn_hair_03_cr, bone2Name=dyn_hair_01_cr, swing1LimitDegrees=8, swing2LimitDegrees=14, twistLimitDegrees=2}
 {bone1Name=dyn_hair_04_cr, bone2Name=dyn_hair_03_cr, swing1LimitDegrees=2, swing2LimitDegrees=12, twistLimitDegrees=2}
 {bone1Name=dyn_hair_05_cr, bone2Name=dyn_hair_04_cr, swing1LimitDegrees=2.5, swing2LimitDegrees=12, twistLimitDegrees=2}
 {bone1Name=dyn_hair_03_lf, bone2Name=dyn_hair_01_lf, swing1LimitDegrees=5, swing2LimitDegrees=14, twistLimitDegrees=2}
 {bone1Name=dyn_hair_04_lf, bone2Name=dyn_hair_03_lf, swing1LimitDegrees=6, swing2LimitDegrees=14, twistLimitDegrees=2}
 {bone1Name=dyn_hair_05_lf, bone2Name=dyn_hair_04_lf, swing1LimitDegrees=5, swing2LimitDegrees=12, twistLimitDegrees=2}
 {bone1Name=dyn_hair_04_sd_l, bone2Name=dyn_hair_03_lf, swing1LimitDegrees=18, swing2LimitDegrees=18, twistLimitDegrees=4}
 {bone1Name=dyn_hair_03_rt, bone2Name=dyn_hair_01_rt, swing1LimitDegrees=8, swing2LimitDegrees=14, twistLimitDegrees=2}
 {bone1Name=dyn_hair_04_rt, bone2Name=dyn_hair_03_rt, swing1LimitDegrees=10, swing2LimitDegrees=14, twistLimitDegrees=2}
 {bone1Name=dyn_hair_05_rt, bone2Name=dyn_hair_04_rt, swing1LimitDegrees=8, swing2LimitDegrees=12, twistLimitDegrees=2}
 {bone1Name=dyn_hair_04_sd_r, bone2Name=dyn_hair_03_rt, swing1LimitDegrees=18, swing2LimitDegrees=18, twistLimitDegrees=4}
 {bone1Name=base_hair_02_rt_helper, bone2Name=dyn_hair_03_rt, swing1LimitDegrees=0, swing2LimitDegrees=0, twistLimitDegrees=0}
 {
  bone1Name=base_hair_02_lt_helper
  bone2Name=head
  swing1LimitDegrees=45
  swing1Motion=Free
  swing2LimitDegrees=45
  swing2Motion=Free
  twistLimitDegrees=45
  twistMotion=Free
 }
 {
  bone1Name=dyn_hair_05_rt
  bone2Name=dyn_hair_05_cr
  swing1LimitDegrees=45
  swing1Motion=Free
  swing2LimitDegrees=45
  swing2Motion=Free
  twistLimitDegrees=45
  twistMotion=Free
 }
 {
  bone1Name=dyn_hair_05_lf
  bone2Name=dyn_hair_05_cr
  swing1LimitDegrees=45
  swing1Motion=Free
  swing2LimitDegrees=45
  swing2Motion=Free
  twistLimitDegrees=45
  twistMotion=Free
 }
]
