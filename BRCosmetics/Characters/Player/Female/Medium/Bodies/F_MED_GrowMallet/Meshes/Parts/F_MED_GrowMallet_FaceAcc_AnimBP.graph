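# Post-process style AnimBP for the GrowMallet female character's face accessory part, animating its hair.
# It copies the parent pose, blends in Body_Mesh shoulders, and runs a Control Rig driven by melee, targeting, skydive and cape-backbling states.
# RigidBody and AnimDynamics hair simulation follows, then constraints to spine_05 and additive bone tweaks feed the output; several side chains are unconnected.
# The EventGraph detects backbling and updates gravity and physics alpha variables.
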AnimBlueprint /BRCosmetics/Characters/Player/Female/Medium/Bodies/F_MED_GrowMallet/Meshes/Parts/F_MED_GrowMallet_FaceAcc_AnimBP
target_skeleton=&/F_MED_GrowMallet_FaceAcc_Skeleton
graphs [
 AnimGraph {nodes [
   n1 = Anim Dynamics - Bone: Hair_Uptail_l_dyn!(ComponentPose, Alpha=1, GravityOverride:Vector="0,0,-980", ExternalForce="0,0,0") -> (Pose) {
    angularDampingOverride=1
    angularSpringConstant=8000
    bAngularSpring=true
    bOverrideAngularDamping=true
    bUseGravityOverride=true
    gravityScale=0.35
    simulationSpace=RootRelative
    boundBone {boneName=Hair_Uptail_l_dyn}
    componentAppliedLinearAccClamp=<3000 3000 3500>
    componentLinearAccScale=<0.3 0.3 0.3>
    componentLinearVelScale=<0.3 0.3 0.3>
    gravityOverride {z=-980}
    physicsBodyDefinitions [{
      boundBone {boneName=Hair_Uptail_l_dyn}
      boxExtents=<20 30 30>
      constraintSetup {angularLimitsMax=<0 -5 15>, angularLimitsMin=<0 -20 -25>, angularTarget=<1 0 0>}
      localJointOffset=<6 0 0>
     }]
    simSpaceSettings {maxAngularAcceleration=100, maxAngularVelocity=10, simSpaceAngularAlpha=1}
   }
   Anim Dynamics - Bone: Hair_Uptail_r_dyn!(ComponentPose=@n1, Alpha=1, GravityOverride:Vector="0,0,-980", ExternalForce="0,0,0") -> (Pose) {
    angularDampingOverride=1
    angularSpringConstant=8000
    bAngularSpring=true
    bOverrideAngularDamping=true
    bUseGravityOverride=true
    gravityScale=0.35
    simulationSpace=RootRelative
    boundBone {boneName=Hair_Uptail_r_dyn}
    componentAppliedLinearAccClamp=<3000 3000 3500>
    componentLinearAccScale=<0.3 0.3 0.3>
    componentLinearVelScale=<0.3 0.3 0.3>
    gravityOverride {z=-980}
    physicsBodyDefinitions [{
      boundBone {boneName=Hair_Uptail_r_dyn}
      boxExtents=<20 30 30>
      constraintSetup {angularLimitsMax=<0 25 15>, angularLimitsMin=<0 5 -25>, angularTarget=<1 0 0>}
      localJointOffset=<6 0 0>
     }]
    simSpaceSettings {maxAngularAcceleration=100, maxAngularVelocity=10, simSpaceAngularAlpha=1}
   }
   n2 = Anim Dynamics - Bone: dyn_hair_03_rt(Alpha=0.65, ExternalForce="0,0,0") {
    angularDampingOverride=1
    angularSpringConstant=25000
    bAngularSpring=true
    bOverrideAngularDamping=true
    bUseGravityOverride=true
    gravityScale=0.35
    simulationSpace=RootRelative
    boundBone {boneName=dyn_hair_03_rt}
    componentAppliedLinearAccClamp=<1000 1000 1000>
    componentLinearAccScale=<0.1 0.1 0.15>
    componentLinearVelScale=<0.1 0.1 0.15>
    gravityOverride {z=-980}
    physicsBodyDefinitions [{
      boundBone {boneName=dyn_hair_03_rt}
      boxExtents=<30 30 30>
      constraintSetup {angularLimitsMax=<0 3 0>, angularLimitsMin=<0 -10 -10>, angularTarget=<1 0 0>}
      localJointOffset=<2 0 0>
     }]
    simSpaceSettings {maxAngularAcceleration=100, maxAngularVelocity=10, simSpaceAngularAlpha=1}
   }
   n3 = Anim Dynamics - Bone: dyn_hair_03_lf(ComponentPose=@n2, Alpha=0.65, ExternalForce="0,0,0") {
    angularDampingOverride=1
    angularSpringConstant=25000
    bAngularSpring=true
    bOverrideAngularDamping=true
    bUseGravityOverride=true
    gravityScale=0.35
    simulationSpace=RootRelative
    boundBone {boneName=dyn_hair_03_lf}
    componentAppliedLinearAccClamp=<1000 1000 1000>
    componentLinearAccScale=<0.1 0.1 0.15>
    componentLinearVelScale=<0.1 0.1 0.15>
    gravityOverride {z=-980}
    physicsBodyDefinitions [{
      boundBone {boneName=dyn_hair_03_lf}
      boxExtents=<30 30 30>
      constraintSetup {angularLimitsMax=<0 3 10>, angularLimitsMin=<0 -10 -2>, angularTarget=<1 0 0>}
      localJointOffset=<2 0 0>
     }]
    simSpaceSettings {maxAngularAcceleration=100, maxAngularVelocity=10, simSpaceAngularAlpha=1}
   }
   Anim Dynamics - Bone: dyn_hair_04_lf(ComponentPose=@n3, Alpha=1, ExternalForce="0,0,0") {
    angularDampingOverride=1
    angularSpringConstant=25000
    bAngularSpring=true
    bOverrideAngularDamping=true
    bUseGravityOverride=true
    gravityScale=0.35
    simulationSpace=RootRelative
    boundBone {boneName=dyn_hair_04_lf}
    componentAppliedLinearAccClamp=<1000 1000 1000>
    componentLinearAccScale=<0.1 0.1 0.15>
    componentLinearVelScale=<0.1 0.1 0.15>
    gravityOverride {z=-980}
    physicsBodyDefinitions [{
      boundBone {boneName=dyn_hair_04_lf}
      boxExtents=<30 30 30>
      constraintSetup {angularLimitsMax=<0 3 10>, angularLimitsMin=<0 -10 -2>, angularTarget=<1 0 0>}
      localJointOffset=<2 0 0>
     }]
    simSpaceSettings {maxAngularAcceleration=100, maxAngularVelocity=10, simSpaceAngularAlpha=1}
   }
   n4 = RigidBody!(
    ComponentPose=Local To Component(LocalPose=Use cached pose 'controlRigCache'())
    Alpha
    OverrideWorldGravity:Vector="0,0,-980"
    SimSpaceSettings:Sim Space Settings Structure=(WorldAlpha=1,VelocityScaleZ=1,DampingAlpha=0,MaxLinearVelocity=350,MaxAngularVelocity=4,MaxLinearAcceleration=1500,MaxAngularAcceleration=9,ExternalLinearDragV=(X=1.1,Y=1.1,Z=0),ExternalLinearVelocity=(X=0,Y=0,Z=0),ExternalAngularVelocity=(X=0,Y=0,Z=0))
   ) -> (Pose) {
    bOverrideWorldGravity=true
    simulationSpace=BaseBoneSpace
    alphaBoolBlend {blendOption=Linear}
    baseBoneRef {boneName=neck_02}
    overrideWorldGravity {z=-980}
    simSpaceSettings {
     dampingAlpha=0
     maxAngularAcceleration=9
     maxAngularVelocity=4
     maxLinearAcceleration=1500
     maxLinearVelocity=350
     worldAlpha=1
     externalLinearDragV=<1.1 1.1>
    }
   }
   Blend Poses by bool(BlendPose_0=Component To Local(ComponentPose=@n4), bActiveValue=Get bIsPlayingEmote())
   Output Pose(Result=Use cached pose 'backpackCapeCache'()) {layerGroup=None}
   n5 = Layered blend per bone(
    BasePose=Copy Pose From Mesh() {bUseAttachedParent=true}
    BlendPoses_0=Copy Pose From Mesh(SourceMeshComponent=Get Body_Mesh()) {bUseAttachedParent=true}
   ) {layerSetup [{branchFilters [{boneName=clavicle_l}, {boneName=clavicle_r}, {boneName=upperarm_r}, {boneName=upperarm_l}]}]}
   n6 = Control Rig!(
    Source=@n5
    Alpha
    IsSwingPickaxe:bool=Get bIsPlayingMeleeAnim()
    isTargeting:bool=Get bIsTargeting()
    isWearingBackPackCape:bool=Get isWearingBackblingCape() -> (isWearingBackblingCape:bool)
    SkyDiving:bool=Get SkyDiveMode() -> (SkyDiveMode:bool)
    ParachuteOpened:bool=Get bIsParachuteOpened()
    SkyFloating:bool=Get bIsSkydiveFloating()
   ) -> (Pose) {controlRigAssetReference {blueprintRigClass=&/F_MED_GrowMallet_FaceAcc_CtrlRig.F_MED_GrowMallet_FaceAcc_CtrlRig_C}}
   Save cached pose 'controlRigCache'(Pose=@n6) {cacheName=controlRigCache}
   n7 = RigidBody(
    ComponentPose=Local To Component(LocalPose=Use cached pose 'controlRigCache'())
    OverrideWorldGravity="0,0,-980"
    SimSpaceSettings=(WorldAlpha=1,VelocityScaleZ=1,DampingAlpha=0,MaxLinearVelocity=350,MaxAngularVelocity=4,MaxLinearAcceleration=1500,MaxAngularAcceleration=9,ExternalLinearDragV=(X=1.1,Y=1.1,Z=0),ExternalLinearVelocity=(X=0,Y=0,Z=0),ExternalAngularVelocity=(X=0,Y=0,Z=0))
   ) {
    bOverrideWorldGravity=true
    overridePhysicsAsset=&/F_MED_GrowMallet_FaceAcc_Physics
    simulationSpace=BaseBoneSpace
    alphaBoolBlend {blendOption=Linear}
    baseBoneRef {boneName=neck_02}
    overrideWorldGravity {z=-980}
    simSpaceSettings {
     dampingAlpha=0
     maxAngularAcceleration=9
     maxAngularVelocity=4
     maxLinearAcceleration=1500
     maxLinearVelocity=350
     worldAlpha=1
     externalLinearDragV=<1.1 1.1>
    }
   }
   n8 = Anim Dynamics - Bone: Hair_tail_l_dyn(ComponentPose=@n7, Alpha=1, ExternalForce="0,0,0") {
    angularDampingOverride=1
    angularSpringConstant=20000
    bAngularSpring=true
    bOverrideAngularDamping=true
    bUseGravityOverride=true
    gravityScale=0.35
    simulationSpace=BoneRelative
    boundBone {boneName=Hair_tail_l_dyn}
    componentLinearAccScale=<1 1 1>
    componentLinearVelScale=<1 1 1>
    gravityOverride {z=-980}
    physicsBodyDefinitions [{
      boundBone {boneName=Hair_tail_l_dyn}
      boxExtents=<20 30 30>
      constraintSetup {angularLimitsMax=<15 15 15>, angularLimitsMin=<-15 -15 -15>, angularTarget=<1 0 0>}
      localJointOffset=<4 0 0>
     }]
    relativeSpaceBone {boneName=dyn_hair_05_cr}
   }
   n9 = Anim Dynamics - Bone: Hair_tail_c_dyn(ComponentPose=@n8, Alpha=1, ExternalForce="0,0,0") {
    angularDampingOverride=1
    angularSpringConstant=20000
    bAngularSpring=true
    bOverrideAngularDamping=true
    bUseGravityOverride=true
    gravityScale=0.35
    simulationSpace=BoneRelative
    boundBone {boneName=Hair_tail_c_dyn}
    componentLinearAccScale=<1 1 1>
    componentLinearVelScale=<1 1 1>
    gravityOverride {z=-980}
    physicsBodyDefinitions [{
      boundBone {boneName=Hair_tail_c_dyn}
      boxExtents=<20 30 30>
      constraintSetup {angularLimitsMax=<15 15 15>, angularLimitsMin=<-15 -15 -15>, angularTarget=<1 0 0>}
      localJointOffset=<4 0 0>
     }]
    relativeSpaceBone {boneName=dyn_hair_05_cr}
   }
   n10 = Anim Dynamics - Bone: Hair_tail_r_dyn(ComponentPose=@n9, Alpha=1, ExternalForce="0,0,0") {
    angularDampingOverride=1
    angularSpringConstant=20000
    bAngularSpring=true
    bOverrideAngularDamping=true
    bUseGravityOverride=true
    gravityScale=0.35
    simulationSpace=BoneRelative
    boundBone {boneName=Hair_tail_r_dyn}
    componentLinearAccScale=<1 1 1>
    componentLinearVelScale=<1 1 1>
    gravityOverride {z=-980}
    physicsBodyDefinitions [{
      boundBone {boneName=Hair_tail_r_dyn}
      boxExtents=<20 30 30>
      constraintSetup {angularLimitsMax=<15 15 15>, angularLimitsMin=<-15 -15 -15>, angularTarget=<1 0 0>}
      localJointOffset=<4 0 0>
     }]
    relativeSpaceBone {boneName=dyn_hair_05_cr}
   }
   n11 = Anim Dynamics - Bone: dyn_hair_01_rt(ComponentPose=@n10, Alpha=0.8, ExternalForce="0,0,0") {
    angularDampingOverride=1
    angularSpringConstant=25000
    bAngularSpring=true
    bOverrideAngularDamping=true
    bUseGravityOverride=true
    gravityScale=0.35
    simulationSpace=RootRelative
    boundBone {boneName=dyn_hair_01_rt}
    componentAppliedLinearAccClamp=<1000 1000 1000>
    componentLinearAccScale=<0.1 0.1 0.15>
    componentLinearVelScale=<0.1 0.1 0.15>
    gravityOverride {z=-980}
    physicsBodyDefinitions [{
      boundBone {boneName=dyn_hair_01_rt}
      boxExtents=<30 30 30>
      constraintSetup {angularLimitsMax=<0 3 2>, angularLimitsMin=<0 -10 -10>, angularTarget=<1 0 0>}
      localJointOffset=<2 0 0>
     }]
    simSpaceSettings {maxAngularAcceleration=100, maxAngularVelocity=10, simSpaceAngularAlpha=1}
   }
   n12 = Anim Dynamics - Bone: dyn_hair_01_cr(ComponentPose=@n11, Alpha=0.8, ExternalForce="0,0,0") {
    angularDampingOverride=1
    angularSpringConstant=25000
    bAngularSpring=true
    bOverrideAngularDamping=true
    bUseGravityOverride=true
    gravityScale=0.35
    simulationSpace=RootRelative
    boundBone {boneName=dyn_hair_01_cr}
    componentAppliedLinearAccClamp=<1000 1000 1000>
    componentLinearAccScale=<0.1 0.1 0.15>
    componentLinearVelScale=<0.1 0.1 0.15>
    gravityOverride {z=-980}
    physicsBodyDefinitions [{
      boundBone {boneName=dyn_hair_01_cr}
      boxExtents=<30 30 30>
      constraintSetup {angularLimitsMax=<0 3 5>, angularLimitsMin=<0 -10 -5>, angularTarget=<1 0 0>}
      localJointOffset=<2 0 0>
     }]
    simSpaceSettings {maxAngularAcceleration=100, maxAngularVelocity=10, simSpaceAngularAlpha=1}
   }
   n13 = Anim Dynamics - Bone: dyn_hair_01_lf(ComponentPose=@n12, Alpha=0.8, ExternalForce="0,0,0") {
    angularDampingOverride=1
    angularSpringConstant=25000
    bAngularSpring=true
    bOverrideAngularDamping=true
    bUseGravityOverride=true
    gravityScale=0.35
    simulationSpace=RootRelative
    boundBone {boneName=dyn_hair_01_lf}
    componentAppliedLinearAccClamp=<1000 1000 1000>
    componentLinearAccScale=<0.1 0.1 0.15>
    componentLinearVelScale=<0.1 0.1 0.15>
    gravityOverride {z=-980}
    physicsBodyDefinitions [{
      boundBone {boneName=dyn_hair_01_lf}
      boxExtents=<30 30 30>
      constraintSetup {angularLimitsMax=<0 3 10>, angularLimitsMin=<0 -10 -2>, angularTarget=<1 0 0>}
      localJointOffset=<2 0 0>
     }]
    simSpaceSettings {maxAngularAcceleration=100, maxAngularVelocity=10, simSpaceAngularAlpha=1}
   }
   n14 = Anim Dynamics - Bone: dyn_hair_04_rt(ComponentPose=@n13, Alpha=0.65, ExternalForce="0,0,0") {
    angularDampingOverride=1
    angularSpringConstant=25000
    bAngularSpring=true
    bOverrideAngularDamping=true
    bUseGravityOverride=true
    gravityScale=0.35
    simulationSpace=RootRelative
    boundBone {boneName=dyn_hair_04_rt}
    componentAppliedLinearAccClamp=<1000 1000 1000>
    componentLinearAccScale=<0.1 0.1 0.15>
    componentLinearVelScale=<0.1 0.1 0.15>
    gravityOverride {z=-980}
    physicsBodyDefinitions [{
      boundBone {boneName=dyn_hair_04_rt}
      boxExtents=<30 30 30>
      constraintSetup {angularLimitsMax=<0 3 2>, angularLimitsMin=<0 -10 -10>, angularTarget=<1 0 0>}
      localJointOffset=<2 0 0>
     }]
    simSpaceSettings {maxAngularAcceleration=100, maxAngularVelocity=10, simSpaceAngularAlpha=1}
   }
   Save cached pose 'dynamicsCache'(Pose=Component To Local(ComponentPose=@n14)) {cacheName=dynamicsCache}
   n15 = "Constraint - dyn_hair_03_lf (P)"!(ComponentPose=Local To Component(LocalPose=Use cached pose 'dynamicsCache'()), bAlphaBoolEnabled:bool=True, ConstraintWeights_0=1) -> (Pose) {alphaInputType=Bool, boneToModify {boneName=dyn_hair_03_lf}, constraintSetup [{targetBone {boneName=spine_05}}], constraintWeights [1]}
   n16 = "Constraint - dyn_hair_04_lf (P)"!(ComponentPose=@n15, bAlphaBoolEnabled:bool=True, ConstraintWeights_0=1) -> (Pose) {alphaInputType=Bool, boneToModify {boneName=dyn_hair_04_lf}, constraintSetup [{targetBone {boneName=spine_05}}], constraintWeights [1]}
   n17 = "Constraint - dyn_hair_05_lf (P)"!(ComponentPose=@n16, bAlphaBoolEnabled:bool=True, ConstraintWeights_0=1) -> (Pose) {alphaInputType=Bool, boneToModify {boneName=dyn_hair_05_lf}, constraintSetup [{targetBone {boneName=spine_05}}], constraintWeights [1]}
   n18 = "Constraint - dyn_hair_03_cr (P)"!(ComponentPose=@n17, bAlphaBoolEnabled:bool=True, ConstraintWeights_0=1) -> (Pose) {alphaInputType=Bool, boneToModify {boneName=dyn_hair_03_cr}, constraintSetup [{targetBone {boneName=spine_05}}], constraintWeights [1]}
   n19 = "Constraint - dyn_hair_04_cr (P)"!(ComponentPose=@n18, bAlphaBoolEnabled:bool=True, ConstraintWeights_0=1) -> (Pose) {alphaInputType=Bool, boneToModify {boneName=dyn_hair_04_cr}, constraintSetup [{targetBone {boneName=spine_05}}], constraintWeights [1]}
   n20 = "Constraint - dyn_hair_05_cr (P)"!(ComponentPose=@n19, bAlphaBoolEnabled:bool=True, ConstraintWeights_0=1) -> (Pose) {alphaInputType=Bool, boneToModify {boneName=dyn_hair_05_cr}, constraintSetup [{targetBone {boneName=spine_05}}], constraintWeights [1]}
   n21 = "Constraint - dyn_hair_03_rt (P)"!(ComponentPose=@n20, bAlphaBoolEnabled:bool=True, ConstraintWeights_0=1) -> (Pose) {alphaInputType=Bool, boneToModify {boneName=dyn_hair_03_rt}, constraintSetup [{targetBone {boneName=spine_05}}], constraintWeights [1]}
   n22 = "Constraint - dyn_hair_04_rt (P)"!(ComponentPose=@n21, bAlphaBoolEnabled:bool=True, ConstraintWeights_0=1) -> (Pose) {alphaInputType=Bool, boneToModify {boneName=dyn_hair_04_rt}, constraintSetup [{targetBone {boneName=spine_05}}], constraintWeights [1]}
   n23 = "Constraint - dyn_hair_04_rt (P)"(ComponentPose=@n22, bAlphaBoolEnabled=True, ConstraintWeights_0=1) {alphaInputType=Bool, boneToModify {boneName=dyn_hair_04_rt}, constraintSetup [{targetBone {boneName=spine_05}}], constraintWeights [1]}
   n24 = "Transform (Modify) Bone - Bone: dyn_hair_03_lf"!(ComponentPose=@n23, bAlphaBoolEnabled:bool=True, Rotation="-5,0,0", Scale="1,1,0.5") -> (Pose) {
    alphaInputType=Bool
    rotationMode=BMM_Additive
    rotationSpace=BCS_BoneSpace
    scaleMode=BMM_Additive
    boneToModify {boneName=dyn_hair_03_lf}
    rotation {pitch=-5}
    scale {z=0.5}
   }
   n25 = "Transform (Modify) Bone - Bone: dyn_hair_03_cr"!(ComponentPose=@n24, bAlphaBoolEnabled:bool=True, Rotation="-5,0,0", Scale="1,1,0.5") -> (Pose) {
    alphaInputType=Bool
    rotationMode=BMM_Additive
    rotationSpace=BCS_BoneSpace
    scaleMode=BMM_Additive
    boneToModify {boneName=dyn_hair_03_cr}
    rotation {pitch=-5}
    scale {z=0.5}
   }
   n26 = "Transform (Modify) Bone - Bone: dyn_hair_03_rt"!(ComponentPose=@n25, bAlphaBoolEnabled:bool=True, Rotation="-5,0,0", Scale="1,1,0.5") -> (Pose) {
    alphaInputType=Bool
    rotationMode=BMM_Additive
    rotationSpace=BCS_BoneSpace
    scaleMode=BMM_Additive
    boneToModify {boneName=dyn_hair_03_rt}
    rotation {pitch=-5}
    scale {z=0.5}
   }
   Save cached pose 'backpackCapeCache'(Pose=Component To Local(ComponentPose=@n26)) {cacheName=backpackCapeCache}
   Get bIsSkyDiving()
  ]}
 EventGraph {nodes [
   Event OnCharacterCustomizationLoaded()
   > n1 = Cast To PlayerPawn_Athena(Object=TryGetPawnOwner())
   > Set Body_Mesh(Body_Mesh=GetSkeletalMeshForPartType(self=@n1))
   n2 = CheckForBackbling()
   n3 = NOT Boolean(A=@n2.bIsBackblingLong)
   n4 = NOT Boolean(A=@n2.bIsBackblingACape)
   > Set isWearingBackblingCape(isWearingBackblingCape:bool=OR Boolean(A=AND Boolean!(A=@n2, B=@n4, C:bool=@n3) -> (ReturnValue), B=OR Boolean(A=@n2.bIsBackblingACape, B=@n2.bIsBackblingLong))) -> (Output_Get:bool)
   Event BlueprintUpdateAnimation()
   > n5 = Sequence!(execute) -> (then_0, then_1, then_2:Exec, then_3:Exec)
   > n6 = "Gravity Override State Function (Parameter Asset) (Nativized)"(
    CharacterPartAnimInstance=Self-Reference()
    GameTimeInSeconds=GetGameTimeInSeconds()
    Parameters=&/F_MED_GrowMallet_FaceAcc_Parameters
   )
   > Set GravityOverride(GravityOverride=@n6)
   > Set RBAN Settings Out(RBAN Settings Out=@n6.OutRigidBodyAnimNodeParameters)
   n7 = Branch(execute=@n5.then_2, Condition=Get bIsSkyDiving())
   > Set AlphaAnim_Dyn(AlphaAnim_Dyn:double=0) -> (Output_Get:double)
   Set AlphaAnim_Dyn(execute=@n7.else, AlphaAnim_Dyn=1)
   n8 = Branch(execute=@n5.then_1, Condition=Get bIsPlayingMeleeAnim())
   > "Set Alpha_HairPhysics "("Alpha_HairPhysics ":double=1) -> (Output_Get:double)
   n9 = Branch(execute=@n8.else, Condition=Get bIsParachuteOpened())
   > "Set Alpha_HairPhysics "("Alpha_HairPhysics "=0.9)
   n10 = Branch(execute=@n9.else, Condition=Get bIsSkyDiving())
   > "Set Alpha_HairPhysics "("Alpha_HairPhysics "=0.8)
   n11 = Branch(execute=@n10.else, Condition=Get bIsCrouching())
   > "Set Alpha_HairPhysics "("Alpha_HairPhysics "=0.8)
   "Set Alpha_HairPhysics "(execute=@n11.else, "Alpha_HairPhysics "=1)
   n12 = Branch(execute=@n5.then_3, Condition=OR Boolean(A=Get bIsSkydiveFloating(), B=Get bIsSkydiveDiveMode()))
   > Set SkyDiveMode(SkyDiveMode:bool=true) -> (Output_Get:bool)
   Set SkyDiveMode(execute=@n12.else, SkyDiveMode=false)
  ]}
]
variables [
 "Alpha_HairPhysics " {type=number, default=1}
 AlphaAnim_Dyn {type=number, default=1}
 Body_Mesh {type=SkeletalMeshComponent}
 GravityOverride {type=Vector}
 isWearingBackblingCape {type=boolean}
 RBAN Settings Out {
  type=FortRigidBodyAnimNodeParameters
  default {simSpaceSettings {
    dampingAlpha=0
    maxAngularAcceleration=30
    maxAngularVelocity=8
    maxLinearAcceleration=1000
    maxLinearVelocity=400
    velocityScaleZ=0.4
    worldAlpha=1
    externalLinearDragV=<0.2 0.2 0.2>
   }}
 }
 SkyDiveMode {type=boolean}
]
